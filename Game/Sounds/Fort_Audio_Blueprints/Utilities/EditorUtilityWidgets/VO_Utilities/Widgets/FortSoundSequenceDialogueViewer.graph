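# Editor utility widget for VO tooling that shows a Fort sound sequence's dialogue lines, or unassigned lines, with a name field and bumper dropdown.
# Row entry edits (sound, delay, reorder, unassign), renames and bumper choices are re-broadcast through the viewer's own dispatchers.
# Play builds a transient FortSoundSequence and chains lines via SpawnSound2D, audio-finished callbacks and delay timers in the PIE or editor world, highlighting the current row.
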EditorUtilityWidgetBlueprint /Game/Sounds/Fort_Audio_Blueprints/Utilities/EditorUtilityWidgets/VO_Utilities/Widgets/FortSoundSequenceDialogueViewer
graphs [
 Advance Sequence {nodes [
   Advance Sequence() {is=title K2Node_FunctionEntry}
   n1 = GetSoundSequenceData!(self:Fort Sound Sequence Object Reference=Get Preview Sound Sequence() -> (Preview Sound Sequence:Fort Sound Sequence Object Reference)) -> (OutSoundSequenceData:Array of Sound Sequence Data Structures)
   > n2 = Branch(Condition="Equal (Integer)"(A=Get Current Playback Index() -> (Current Playback Index:int=0), B=Length(TargetArray:Array of Sound Sequence Data Structures=@n1)))
   > Cleanup Playback()
   Increment Int(" "=@n2.else, Value=Get Current Playback Index() -> (Current Playback Index=0))
   n3 = Get Preview Sound Sequence()
   > n4 = Branch(Condition=Is Valid Index!(
     TargetArray:Array of Sound Sequence Data Structures=GetSoundSequenceData(self=@n3)
     IndexToTest:int=Get Current Playback Index() -> (Current Playback Index=0)
    ) -> (ReturnValue:bool))
   > Play Current Line()
   Cleanup Playback(execute=@n4.else)
  ]}
 Cleanup Playback {nodes [
   Cleanup Playback() {is=title K2Node_FunctionEntry}
   > Set Current Playback Index(Current Playback Index:int=0) -> (Output_Get:int=0)
   > Set Preview Sound Sequence(Preview Sound Sequence:Fort Sound Sequence Object Reference) -> (Output_Get:Fort Sound Sequence Object Reference)
   > Clear and Invalidate Timer by Handle!(execute, WorldContextObject:Object Reference, Handle=Get Current Delay Timer() -> (Current Delay Timer:Timer Handle Structure)) -> (then)
   > Set Is Now Playing!(
    execute
    Is Now Playing:bool=false
    self:Fort Dialogue Data Object Reference=Get Currently Playing Dialogue Data() -> (Currently Playing Dialogue Data:Fort Dialogue Data Object Reference)
   ) -> (then, Output_Get:bool)
   > SetIsEnabled(self=Get DialogueListView() -> (DialogueListView:Editor Utility List View Object Reference), bInIsEnabled=true)
   > Call On Preview Playback Finished!(execute:Exec, self:Fort Sound Sequence Dialogue Viewer Object Reference) -> (then:Exec) {is=title K2Node_CallDelegate}
  ]}
 Create Preview Sound Sequence {nodes [
   Create Preview Sound Sequence() {is=title K2Node_FunctionEntry}
   > n1 = Construct Object from Class(Class=/Script/FortniteGame.FortSoundSequence) -> (ReturnValue:Fort Sound Sequence Object Reference)
   > n2 = Set Preview Sound Sequence(Preview Sound Sequence=@n1)
   > Populate Sound Sequence Data!(
    execute
    Sound Sequence:Fort Sound Sequence Object Reference=@n2
    In Sound Sequence Data:Fort Sound Sequence Data Object Reference=Get Sound Sequence Data() -> (Sound Sequence Data:Fort Sound Sequence Data Object Reference)
    __WorldContext:Object Reference
   ) -> (then)
  ]}
 Dialogue Data Contains At Least One Sound {nodes [
   Dialogue Data Contains At Least One Sound() {is=title K2Node_FunctionEntry}
   > n1 = For Each Loop with Break!(
    Exec:Exec
    Array:Array of Fort Dialogue Data Object References=Get Dialogue Data!(self:Fort Sound Sequence Data Object Reference=Get Sound Sequence Data()) -> (Dialogue Data:Array of Fort Dialogue Data Object References) {is=title K2Node_CallFunction}
    Break:Exec=@n3.then
   ) -> (LoopBody:Exec, Array Element:Fort Dialogue Data Object Reference, Array Index:int, Completed:Exec)
   Return Node!(execute=@n1.Completed, ReturnValue:bool=Get Contains At Least One Sound() -> (Contains At Least One Sound:bool))
   > n2 = Get!(execute:Exec, self:Fort Dialogue Data Object Reference=@n1) -> (then:Exec, else:Exec, Sound:Sound Base Object Reference)
   > n3 = Set Contains At Least One Sound(Contains At Least One Sound:bool=true) -> (Output_Get:bool)
   Set Contains At Least One Sound(execute=@n2.else, Contains At Least One Sound=false)
  ]}
 EventGraph {nodes [
   n1 = "On Entry Generated (DialogueListView)"!() -> (OutputDelegate, then, Widget:User Widget Object Reference) {is=title K2Node_ComponentBoundEvent}
   > n2 = Cast To FortSoundSequenceDialogueEntry!(execute, Object=@n1.Widget) -> (then, CastFailed, AsFort Sound Sequence Dialogue Entry:Fort Sound Sequence Dialogue Entry Object Reference) {is=title K2Node_DynamicCast}
   n3 = Dialogue Data Changed!() -> (OutputDelegate, then, Dialogue Data Object:Fort Dialogue Data Object Reference, New Delay Amount:double=~) {is=title K2Node_CustomEvent}
   n4 = Bind Event to On Delay Amount Changed(execute=@n2.then, self:Fort Sound Sequence Dialogue Entry Object Reference=@n2, Delegate=@n3) {is=title K2Node_AddDelegate}
   n5 = Sound Changed!() -> (OutputDelegate, then, Dialogue Data Object:Fort Dialogue Data Object Reference, New Sound:Sound Base Object Reference) {is=title K2Node_CustomEvent}
   n6 = Bind Event to On Sound Changed(execute=@n4.then, self:Fort Sound Sequence Dialogue Entry Object Reference=@n2, Delegate=@n5) {is=title K2Node_AddDelegate}
   n7 = Unassign Line!() -> (OutputDelegate, then, Dialogue Data Object:Fort Dialogue Data Object Reference) {is=title K2Node_CustomEvent}
   n8 = Bind Event to On Line Unassigned(execute=@n6.then, self:Fort Sound Sequence Dialogue Entry Object Reference=@n2, Delegate=@n7) {is=title K2Node_AddDelegate}
   n9 = Reorder List!() -> (
    OutputDelegate
    then
    Dialogue Data to Insert:Fort Dialogue Data Object Reference
    Dialogue Data to Insert Above:Fort Dialogue Data Object Reference
   ) {is=title K2Node_CustomEvent}
   Bind Event to On Lines Reordered(execute=@n8.then, self:Fort Sound Sequence Dialogue Entry Object Reference=@n2, Delegate=@n9) {is=title K2Node_AddDelegate}
   n10 = "On Entry Released (DialogueListView)"!() -> (OutputDelegate, then, Widget:User Widget Object Reference) {is=title K2Node_ComponentBoundEvent}
   > n11 = Cast To FortSoundSequenceDialogueEntry(Object=@n10.Widget)
   > Reset(self:Fort Sound Sequence Dialogue Entry Object Reference=@n11)
   n12 = "OnTextChanged (Editable Text) (EditableAssetName)"!() -> (OutputDelegate, then, Text:"Text (by ref)") {is=title K2Node_ComponentBoundEvent}
   > Call On Sound Sequence Name Changed!(execute:Exec, self:Fort Sound Sequence Dialogue Viewer Object Reference, Sound Sequence Name:String="To String (Text)"(InText=@n12.Text)) -> (then:Exec) {is=title K2Node_CallDelegate}
   n13 = "OnTextCommitted (Editable Text) (EditableAssetName)"!() -> (OutputDelegate, then, Text:"Text (by ref)", CommitMethod:ETextCommit Enum=Default) {is=title K2Node_ComponentBoundEvent}
   > Call On Sound Sequence Name Changed(Sound Sequence Name=MakeValidFileName!(InString:String="To String (Text)"(InText=@n13.Text), InReplacementChar:String) -> (ReturnValue:String))
   "On Clicked (PlayButton)"() {is=title K2Node_ComponentBoundEvent}
   > Create Preview Sound Sequence() {is=title K2Node_CallFunction}
   > SetSubtitlesEnabled!(execute, bEnabled:bool=true) -> (then)
   > Play Current Line()
   > Clear and Invalidate Timer by Handle(Handle=Get Current Delay Timer())
   > SetIsEnabled(self=Get DialogueListView(), bInIsEnabled=false)
   > Call On Preview Playback Started!(execute:Exec, self:Fort Sound Sequence Dialogue Viewer Object Reference) -> (then:Exec) {is=title K2Node_CallDelegate}
   n14 = "On Clicked (StopButton)"() {is=title K2Node_ComponentBoundEvent}
   Unbind Event from On Audio Finished(
    execute=[@n14.then, @n24.then]
    self:Audio Component Object Reference=Get Current Audio Component() -> (Current Audio Component:Audio Component Object Reference)
    Delegate=Create Event()
   ) {is=title K2Node_RemoveDelegate}
   > Stop(self:Audio Component Object Reference=Get Current Audio Component())
   > Cleanup Playback() {is=title K2Node_CallFunction}
   "On Clicked (BrowseToDataAssetButton)"() {is=title K2Node_ComponentBoundEvent}
   > Call On Bumper Settings Clicked!(execute:Exec, self:Fort Sound Sequence Dialogue Viewer Object Reference) -> (then:Exec) {is=title K2Node_CallDelegate}
   n15 = "On Selection Changed (DialogueWrapperComboBox)"!() -> (OutputDelegate, then, SelectedItem:String, SelectionType:ESelectInfo Enum) {is=title K2Node_ComponentBoundEvent}
   > Call On Bumper Selection Changed!(execute:Exec, self:Fort Sound Sequence Dialogue Viewer Object Reference, Selected Bumper:String=@n15.SelectedItem) -> (then:Exec) {is=title K2Node_CallDelegate}
   "On Clicked (ReloadBumperData)"() {is=title K2Node_ComponentBoundEvent}
   > ClearOptions(self:"Combo Box (String) Object Reference"=Get DialogueWrapperComboBox() -> (DialogueWrapperComboBox:Editor Utility Combo Box String Object Reference))
   > n16 = Get!(execute) -> (then, else, Dialogue Bumpers Data Asset:Fort Dialogue Bumpers Data Object Reference)
   n17 = Get Bumper Keys!(self:Fort Dialogue Bumpers Data Object Reference=@n16) -> (Wrapper Keys:Array of Names) {is=title K2Node_CallFunction}
   > n18 = For Each Loop(Array:Array of Names=@n17) -> (Array Element:Name)
   n19 = Get Dialogue Bumpers!(self:Fort Sound Sequence Data Object Reference=Get Sound Sequence Data()) -> (Dialogue Bumpers:Name=None)
   SetSelectedOption!(
    execute=@n18.Completed
    self:"Combo Box (String) Object Reference"=Get DialogueWrapperComboBox()
    Option:String=Select(Option 0:String=No Bumpers, Option 1:String="To String (Name)"(InName=@n19), Index=Contains Item(TargetArray=@n17, ItemToFind=@n19)) -> (ReturnValue:String)
   ) -> (then)
   AddOption!(
    execute=@n18.LoopBody
    self:"Combo Box (String) Object Reference"=Get DialogueWrapperComboBox()
    Option:String="To String (Name)"(InName=@n18)
   ) -> (then)
   n20 = Set Sound Sequence!() -> (OutputDelegate, then, Sound Sequence Data:Fort Sound Sequence Data Object Reference) {is=title K2Node_CustomEvent}
   > Set Sound Sequence Data(Sound Sequence Data:Fort Sound Sequence Data Object Reference=@n20.Sound Sequence Data) -> (Output_Get:Fort Sound Sequence Data Object Reference)
   > Refresh List()
   n21 = Get Sound Sequence Data()
   > SetSelectedOption(self=Get DialogueWrapperComboBox(), Option="To String (Name)"(InName=Get Dialogue Bumpers(self=@n21) -> (Dialogue Bumpers=None)))
   Refresh List() {is=title K2Node_CustomEvent}
   > ClearListItems(self=Get DialogueListView())
   > Filter Dialogue List Entries()
   > Set List Items!(
    execute
    self:List View Object Reference=Get DialogueListView()
    InListItems:Array of Object References=Get Filtered Dialogue Script Objects() -> (Filtered Dialogue Script Objects:Array of Fort Dialogue Data Object References)
   ) -> (then) {is=title K2Node_CallFunction}
   n22 = Set Dialogue Bumpers Data Asset!() -> (OutputDelegate, then, Dialogue Bumpers Data Asset:Fort Dialogue Bumpers Data Object Reference) {is=title K2Node_CustomEvent}
   > Set Dialogue Bumpers Data Asset(Dialogue Bumpers Data Asset:Fort Dialogue Bumpers Data Object Reference=@n22.Dialogue Bumpers Data Asset) -> (Output_Get:Fort Dialogue Bumpers Data Object Reference) {is=title K2Node_VariableSet}
   > n23 = For Each Loop(Array=Get Bumper Keys(self=@n22.Dialogue Bumpers Data Asset))
   > AddOption(self=Get DialogueWrapperComboBox(), Option="To String (Name)"(InName=@n23))
   n24 = Event Destruct()
   Call On Line Unassigned!(
    execute:Exec=@n7.then
    self:Fort Sound Sequence Dialogue Viewer Object Reference
    Dialogue Data Object:Fort Dialogue Data Object Reference=@n7.Dialogue Data Object
   ) -> (then:Exec) {is=title K2Node_CallDelegate}
   Call On Delay Amount Changed!(
    execute:Exec=@n3.then
    self:Fort Sound Sequence Dialogue Viewer Object Reference
    Dialogue Data Object:Fort Dialogue Data Object Reference=@n3.Dialogue Data Object
    New Delay Amount:double=@n3.New Delay Amount
   ) -> (then:Exec) {is=title K2Node_CallDelegate}
   Call On Lines Reordered!(
    execute:Exec=@n9.then
    self:Fort Sound Sequence Dialogue Viewer Object Reference
    Dialogue Data to Insert:Fort Dialogue Data Object Reference=@n9.Dialogue Data to Insert
    Dialogue Data to Insert Above:Fort Dialogue Data Object Reference=@n9.Dialogue Data to Insert Above
   ) -> (then:Exec) {is=title K2Node_CallDelegate}
   Call On Sound Changed!(
    execute:Exec=@n5.then
    self:Fort Sound Sequence Dialogue Viewer Object Reference
    Dialogue Data Object:Fort Dialogue Data Object Reference=@n5.Dialogue Data Object
    New Sound:Sound Base Object Reference=@n5.New Sound
   ) -> (then:Exec) {is=title K2Node_CallDelegate}
  ]}
 Filter Dialogue List Entries {nodes [
   Filter Dialogue List Entries() {is=title K2Node_FunctionEntry}
   > Clear(TargetArray:Array of Fort Dialogue Data Object References=Get Filtered Dialogue Script Objects())
   > n1 = Get!(execute) -> (then, else, Sound Sequence Data:Fort Sound Sequence Data Object Reference)
   > n2 = For Each Loop(Array:Array of Fort Dialogue Data Object References=Get Dialogue Data(self=@n1)) -> (Array Element:Fort Dialogue Data Object Reference)
   > Set Local Dialogue Data(Local Dialogue Data:Fort Dialogue Data Object Reference=@n2) -> (Output_Get:Fort Dialogue Data Object Reference)
   > Set Row Index!(
    execute
    Row Index:int=Length(TargetArray:Array of Fort Dialogue Data Object References=Get Filtered Dialogue Script Objects())
    self:Fort Dialogue Data Object Reference=Get Local Dialogue Data() -> (Local Dialogue Data:Fort Dialogue Data Object Reference)
   ) -> (then, Output_Get:int=0)
   n3 = Get Character!(self:Fort Dialogue Data Object Reference=Get Local Dialogue Data()) -> (Character:String)
   n4 = "Not Equal Exactly (String)"(A=@n3, B=End Sound)
   n5 = "Not Equal Exactly (String)"(A=@n3, B=Start Sound)
   n6 = Get Asset Name!(self:Fort Sound Sequence Data Object Reference=Get Sound Sequence Data()) -> (Asset Name:String)
   > Set Enable Unassign Button!(
    execute
    Enable Unassign Button:bool=AND Boolean(A="Not Equal (Name)"(A:String=@n6, B=Unassigned Lines) -> (ReturnValue=~), B=AND Boolean(A=@n5, B=@n4))
    self:Fort Dialogue Data Object Reference=Get Local Dialogue Data()
   ) -> (then, Output_Get:bool)
   > Add(
    TargetArray:Array of Fort Dialogue Data Object References=Get Filtered Dialogue Script Objects()
    NewItem:"Fort Dialogue Data Object Reference (by ref)"=Get Local Dialogue Data()
   ) {is=title K2Node_CallArrayFunction}
  ]}
 Get Current Dialogue Data {nodes [
   Get Current Dialogue Data() {is=title K2Node_FunctionEntry}
   > n1 = Get!(execute) -> (then, else, Preview Sound Sequence:Fort Sound Sequence Object Reference)
   n2 = GetSoundSequenceData(self=@n1)
   > Branch(Condition=Is Not Empty!(TargetArray:Array of Sound Sequence Data Structures=@n2) -> (ReturnValue:bool))
   > Return Node!(
    execute
    Item:Sound Sequence Data Structure="Get (a copy)"(Array:Array of Sound Sequence Data Structures=@n2, Dimension 1=Get Current Playback Index() -> (Current Playback Index=0)) -> (Output:Sound Sequence Data Structure)
   )
  ]}
 Get Icon Brush {nodes [
   Get Icon Brush() {is=title K2Node_FunctionEntry}
   n1 = Get Asset Name(self=Get Sound Sequence Data())
   > Return Node!(
    execute
    ReturnValue:Slate Brush Structure=Select(
     Option 0:Slate Brush Structure=Make Slate Brush!(
      TintColor:Slate Color Structure=(SpecifiedColor=(R=1,G=1,B=1,A=1),ColorUseRule=UseColor_Specified)
      DrawAs:ESlateBrushDrawType Enum=Image
      Tiling:ESlateBrushTileType Enum=NoTile
      Mirroring:ESlateBrushMirrorType Enum=NoMirror
      ImageSize:Vector 2D Structure=(X=32,Y=32)
      Margin:Margin Structure=(Left=0,Top=0,Right=0,Bottom=0)
      ResourceObject:Object Reference=/Game/Athena/UI/InGame/Creative/PropertyEditor/Icon_Picker/Textures_V2/64x64/T_UI_IconLibrary_Chat_64
      OutlineSettings:Slate Brush Outline Settings Structure=(CornerRadii=(X=0,Y=0,Z=0,W=0),Color=(SpecifiedColor=(R=0,G=0,B=0,A=0),ColorUseRule=UseColor_Specified),Width=0,RoundingType=HalfHeightRadius,bUseBrushTransparency=False)
     ) -> (SlateBrush:Slate Brush Structure) {is=title K2Node_MakeStruct}
     Option 1:Slate Brush Structure=Make Slate Brush(
      TintColor=(SpecifiedColor=(R=1,G=1,B=1,A=1),ColorUseRule=UseColor_Specified)
      DrawAs=Image
      Tiling=NoTile
      Mirroring=NoMirror
      ImageSize=(X=32,Y=32)
      Margin=(Left=0,Top=0,Right=0,Bottom=0)
      ResourceObject=/FortniteConversationUI/Conversation/Materials/T_UI_NPC_SpeechMax
      OutlineSettings=(CornerRadii=(X=0,Y=0,Z=0,W=0),Color=(SpecifiedColor=(R=0,G=0,B=0,A=0),ColorUseRule=UseColor_Specified),Width=0,RoundingType=HalfHeightRadius,bUseBrushTransparency=False)
     )
     Index="Equal (Name)"(A:String=@n1, B=Unassigned Lines)
    ) -> (ReturnValue:Slate Brush Structure)
   )
  ]}
 Get Is Play Button Enabled {nodes [
   Get Is Play Button Enabled() {is=title K2Node_FunctionEntry}
   n1 = IsPlaying!(self:Audio Component Object Reference=Get Current Audio Component()) -> (ReturnValue:bool)
   n2 = Is Timer Active by Handle!(WorldContextObject:Object Reference, Handle:Timer Handle Structure=Get Current Delay Timer()) -> (ReturnValue:bool)
   > Return Node(ReturnValue:bool=AND Boolean!(
     A=NOT Boolean(A=@n1)
     B=NOT Boolean(A=@n2)
     C:bool=Dialogue Data Contains At Least One Sound!() -> (ReturnValue:bool) {is=title K2Node_CallFunction}
    ) -> (ReturnValue))
  ]}
 Get Is Stop Button Enabled {nodes [
   Get Is Stop Button Enabled() {is=title K2Node_FunctionEntry}
   n1 = Get Current Audio Component()
   n2 = Get Current Delay Timer()
   > Return Node(ReturnValue=OR Boolean(A=IsPlaying(self=@n1), B=Is Timer Active by Handle(Handle=@n2)))
  ]}
 Get Is Title Visible {nodes [
   Get Is Title Visible() {is=title K2Node_FunctionEntry}
   n1 = Get Asset Name(self=Get Sound Sequence Data())
   > Return Node(ReturnValue:ESlateVisibility Enum=Select(Option 0:ESlateVisibility Enum=Visible, Option 1:ESlateVisibility Enum=Collapsed, Index="Equal (Name)"(A=@n1, B=Unassigned Lines)) -> (ReturnValue:ESlateVisibility Enum))
  ]}
 Get List Visibility {nodes [
   Get List Visibility() {is=title K2Node_FunctionEntry}
   n1 = GetNumItems!(self:List View Object Reference=Get DialogueListView()) -> (ReturnValue:int=0)
   > Return Node(ReturnValue=Select(Option 0=Collapsed, Option 1=Visible, Index="integer > integer"(A=@n1)))
  ]}
 Get Play Icon Color {nodes [
   Get Play Icon Color() {is=title K2Node_FunctionEntry}
   n1 = IsPlaying(self=Get Current Audio Component())
   n2 = Is Timer Active by Handle(Handle=Get Current Delay Timer())
   > Return Node(ReturnValue:Linear Color Structure=SelectColor(A=(R=0.5029,G=0.5029,B=0.5029,A=1), B=(R=0.2582,G=0.5395,B=0.0685,A=1), bPickA=OR Boolean(A=@n1, B=@n2)))
  ]}
 Get Status Message Text {nodes [
   Get Status Message Text() {is=title K2Node_FunctionEntry}
   n1 = Get Sound Sequence Data()
   n2 = "integer > integer"(A=Length(TargetArray=Get Dialogue Data(self=@n1)))
   > Return Node(ReturnValue:Text=SelectText(
     A=SelectText(B="LOCTEXT(\"There are no dialogue lines in this sound sequence.\")", bPickA=@n2)
     B="LOCTEXT(\"No sound sequence asset selected.\")"
     bPickA=IsValid(Object=@n1)
    ))
  ]}
 Get Status Message Visibility {nodes [
   Get Status Message Visibility() {is=title K2Node_FunctionEntry}
   n1 = NOT Boolean(A=IsValid(Object=Get Sound Sequence Data()))
   n2 = "Equal (Integer)"(A=GetNumItems(self=Get DialogueListView()) -> (ReturnValue=0))
   > Return Node(ReturnValue:ESlateVisibility Enum=Select(Option 0=Collapsed, Option 1=Visible, Index=OR Boolean(A=@n1, B=@n2)))
  ]}
 Get Stop Icon Color {nodes [
   Get Stop Icon Color() {is=title K2Node_FunctionEntry}
   n1 = IsPlaying(self=Get Current Audio Component())
   n2 = Is Timer Active by Handle(Handle=Get Current Delay Timer())
   > Return Node(ReturnValue:Linear Color Structure=SelectColor(A=(R=0.5395,G=0.0685,B=0.0685,A=1), B=(R=0.5029,G=0.5029,B=0.5029,A=1), bPickA=OR Boolean(A=@n1, B=@n2)))
  ]}
 Get Title Text {nodes [
   Get Title Text() {is=title K2Node_FunctionEntry}
   n1 = Get Sound Sequence Data()
   n2 = Get Asset Name(self=@n1)
   n3 = Length(TargetArray=Get Dialogue Data(self=@n1))
   n4 = Select(Option 0:Text="LOCTEXT(\"s\")", Option 1:Text, Index="Equal (Integer)"(A=@n3, B=1)) -> (ReturnValue:Text)
   n5 = Select(Option 1=Format Text!(Format="LOCTEXT(\"{num} \")", num:int=@n3) -> (Result), Index="integer > integer"(A=@n3))
   > Return Node(ReturnValue:Text=Select(
     Option 0=Format Text!(Format="LOCTEXT(\"{num}Unassigned Line{s}\")", num:Text=@n5, s:Text=@n4) -> (Result)
     Option 1="To Text (String)"(InString=@n2)
     Index="Not Equal Exactly (String)"(A=@n2, B=Unassigned Lines)
    ))
  ]}
 Get Unassigned Lines Title Visibility {nodes [
   Get Unassigned Lines Title Visibility() {is=title K2Node_FunctionEntry}
   n1 = Get Asset Name(self=Get Sound Sequence Data())
   > Return Node(ReturnValue:ESlateVisibility Enum=Select(Option 0:ESlateVisibility Enum=Collapsed, Option 1:ESlateVisibility Enum=Visible, Index="Equal (Name)"(A=@n1, B=Unassigned Lines)) -> (ReturnValue:ESlateVisibility Enum))
  ]}
 Get Visibility {nodes [
   Get Visibility() {is=title K2Node_FunctionEntry}
   n1 = Get Sound Sequence Data()
   > Return Node(ReturnValue=Select(Option 0=Collapsed, Option 1=Visible, Index=IsValid(Object=@n1)))
  ]}
 Get World {nodes [
   Get World() {is=title K2Node_FunctionEntry}
   > n1 = GetPIEWorlds!(execute, bIncludeDedicatedServer:bool=false) -> (then, ReturnValue:Array of World Object References)
   > n2 = Branch(Condition=Is Valid Index(TargetArray:Array of World Object References=@n1))
   > Return Node(ReturnValue:World Object Reference="Get (a copy)"(Array:Array of World Object References=@n1) -> (Output:World Object Reference))
   n3 = GetEditorWorld(execute=@n2.else, self=Get UnrealEditorSubsystem())
   > Is Valid(InputObject=@n3)
   > Return Node(ReturnValue=@n3)
  ]}
 Handle Delay {nodes [
   Handle Delay() {is=title K2Node_FunctionEntry}
   > Set Is Now Playing(Is Now Playing=false, self=Get Currently Playing Dialogue Data())
   n1 = Get Current Dialogue Data!() -> (Item_Sound:Sound Base Object Reference, Item_Delay:float=~) {is=title K2Node_CallFunction}
   > n2 = Branch(Condition="float > float"(A=@n1.Item_Delay))
   > n3 = Set Timer by Event(Delegate=Create Event(), Time=@n1.Item_Delay, bLooping=false)
   > Set Current Delay Timer(Current Delay Timer:Timer Handle Structure=@n3) -> (Output_Get:Timer Handle Structure)
   Advance Sequence(execute=@n2.else) {is=title K2Node_CallFunction}
  ]}
 On Bumper Selection Changed {nodes [On Bumper Selection Changed!() -> (then, Selected Bumper:String) {is=title K2Node_FunctionEntry}]}
 On Bumper Settings Clicked {nodes [On Bumper Settings Clicked() {is=title K2Node_FunctionEntry}]}
 On Delay Amount Changed {nodes [On Delay Amount Changed!() -> (then, Dialogue Data Object:Fort Dialogue Data Object Reference, New Delay Amount:double=~) {is=title K2Node_FunctionEntry}]}
 On Line Unassigned {nodes [On Line Unassigned!() -> (then, Dialogue Data Object:Fort Dialogue Data Object Reference) {is=title K2Node_FunctionEntry}]}
 On Lines Reordered {nodes [On Lines Reordered!() -> (then, Dialogue Data to Insert:Fort Dialogue Data Object Reference, Dialogue Data to Insert Above:Fort Dialogue Data Object Reference) {is=title K2Node_FunctionEntry}]}
 On Preview Playback Finished {nodes [On Preview Playback Finished() {is=title K2Node_FunctionEntry}]}
 On Preview Playback Started {nodes [On Preview Playback Started() {is=title K2Node_FunctionEntry}]}
 On Sound Changed {nodes [On Sound Changed!() -> (then, Dialogue Data Object:Fort Dialogue Data Object Reference, New Sound:Sound Base Object Reference) {is=title K2Node_FunctionEntry}]}
 On Sound Sequence Name Changed {nodes [On Sound Sequence Name Changed!() -> (then, Sound Sequence Name:String) {is=title K2Node_FunctionEntry}]}
 Play Current Line {nodes [
   Play Current Line() {is=title K2Node_FunctionEntry}
   > n1 = Set Sound(Sound:Sound Base Object Reference=Get Current Dialogue Data() -> (Item_Delay=~)) -> (Output_Get:Sound Base Object Reference)
   > n2 = Is Valid(InputObject=@n1)
   Cleanup Playback(execute=@n2.Is Not Valid)
   n3 = SpawnSound2D!(
    execute=@n2.Is Valid
    WorldContextObject:Object Reference=Get World!() -> (ReturnValue:World Object Reference) {is=title K2Node_CallFunction}
    Sound:Sound Base Object Reference=Get Sound() -> (Sound:Sound Base Object Reference)
    VolumeMultiplier:float=1
    PitchMultiplier:float=1
    StartTime:float=0
    ConcurrencySettings:Sound Concurrency Object Reference
    bPersistAcrossLevelTransition:bool=false
    bAutoDestroy:bool=true
   ) -> (then, ReturnValue:Audio Component Object Reference)
   > n4 = Set Current Audio Component(Current Audio Component:Audio Component Object Reference=@n3) -> (Output_Get:Audio Component Object Reference)
   > Bind Event to On Audio Finished(self:Audio Component Object Reference=@n4, Delegate=Create Event()) {is=title K2Node_AddDelegate}
   > n5 = For Each Loop with Break(Array=Get Filtered Dialogue Script Objects(), Break=@n7.then)
   Navigate To Item!(execute=@n5.Completed, self:List View Object Reference=Get DialogueListView(), Item:Object Reference=Get Currently Playing Dialogue Data()) -> (then)
   Branch(
    execute=@n5.LoopBody
    Condition="Equal (Object)"(A=Get Sound!(self:Fort Dialogue Data Object Reference=@n5) -> (Sound), B=Get Sound!() -> (Sound))
   )
   > n6 = Set Currently Playing Dialogue Data(Currently Playing Dialogue Data:Fort Dialogue Data Object Reference=@n5) -> (Output_Get:Fort Dialogue Data Object Reference)
   > n7 = Set Is Now Playing(Is Now Playing=true, self=@n6)
  ]}
]
variables [
 Current Audio Component {type=AudioComponent, default=null}
 Current Delay Timer {type=TimerHandle}
 Current Playback Index {type=integer}
 Currently Playing Dialogue Data {type=FortDialogueData_C, default=null}
 Dialogue Bumpers Data Asset {type=FortDialogueBumpersData_C, default=null}
 Filtered Dialogue Script Objects {type="array<FortDialogueData_C>", default []}
 On Bumper Selection Changed {default=()}
 On Bumper Settings Clicked {default=()}
 On Delay Amount Changed {default=()}
 On Line Unassigned {default=()}
 On Lines Reordered {default=()}
 On Preview Playback Finished {default=()}
 On Preview Playback Started {default=()}
 On Sound Changed {default=()}
 On Sound Sequence Name Changed {default=()}
 Preview Sound Sequence {type=FortSoundSequence, default=null}
 Sound Sequence Data {type=FortSoundSequenceData_C, default=null}
]
widgets {
 info {rootWidgetClass=&Border, widgetCount=34}
 widgets [
  Border_3 {widgetClassPath=&Border, widget=Border_3, properties {brushColor=<rgba 0.0144 0.0144 0.0144 1>, padding=<ltrb 0 0 0 0>}}
  VerticalBox_19 {
   widgetClassPath=&VerticalBox
   parent=Border_3
   widget=VerticalBox_19
   slot {class=BorderSlot, properties {padding=<ltrb 0 0 0 0>}}
  }
  SizeBox_1 {
   widgetClassPath=&SizeBox
   parent=VerticalBox_19
   widget=SizeBox_1
   properties {bOverride_HeightOverride=true, heightOverride=64}
   slot {class=VerticalBoxSlot}
  }
  HorizontalBox_75 {
   widgetClassPath=&HorizontalBox
   parent=SizeBox_1
   widget=HorizontalBox_75
   slot {class=SizeBoxSlot, properties {verticalAlignment=VAlign_Center}}
  }
  VerticalBox {
   widgetClassPath=&VerticalBox
   parent=HorizontalBox_75
   widget=VerticalBox
   slot {class=HorizontalBoxSlot, properties {verticalAlignment=VAlign_Top, padding=<ltrb 4 10 0 0>}}
  }
  SizeBox_0 {
   widgetClassPath=&SizeBox
   parent=VerticalBox
   widget=SizeBox_0
   properties {bOverride_HeightOverride=true, bOverride_WidthOverride=true, heightOverride=32, widthOverride=32}
   slot {class=VerticalBoxSlot, properties {horizontalAlignment=HAlign_Center, padding=<ltrb 4 0 4 4>}}
  }
  SoundIcon {
   widgetClassPath=&Image
   parent=SizeBox_0
   widget=SoundIcon
   properties {
    brush {
     bIsDynamicallyLoaded=false
     drawAs=Image
     imageType=FullColor
     mirroring=NoMirror
     resourceName=None
     resourceObject=&T_UI_IconLibrary_Chat_64
     tiling=NoTile
     imageSize=<64 64>
     margin=<ltrb 0 0 0 0>
     outlineSettings {
      bUseBrushTransparency=false
      roundingType=HalfHeightRadius
      width=0
      color {colorUseRule=UseColor_Specified, specifiedColor=<rgba 0 0 0 0>}
      cornerRadii=<0 0 0 0>
     }
     tintColor {colorUseRule=UseColor_Specified, specifiedColor=<rgba 1 1 1 1>}
     uVRegion {bIsValid=false, max=<0 0>, min=<0 0>}
    }
    colorAndOpacity=<rgba 0.5029 0.5029 0.5029 1>
   }
   slot {class=SizeBoxSlot, properties {verticalAlignment=VAlign_Center}}
  }
  VerticalBox_0 {
   widgetClassPath=&VerticalBox
   parent=HorizontalBox_75
   widget=VerticalBox_0
   slot {class=HorizontalBoxSlot, properties {verticalAlignment=VAlign_Center, padding=<ltrb 4 8 8 8>, size {sizeRule=Fill, value=1}}}
  }
  EditableAssetName {
   widgetClassPath=/Script/Blutility.EditorUtilityEditableText
   parent=VerticalBox_0
   widget=EditableAssetName
   properties {
    hintText=Enter a name for this sound sequence asset
    text=Sound Sequence Asset
    widgetStyle {
     backgroundImageComposing {
      bIsDynamicallyLoaded=false
      drawAs=Border
      imageType=FullColor
      mirroring=NoMirror
      resourceName=../../../Engine/Content/Slate/Old/HyperlinkDotted.png
      resourceObject=null
      tiling=Both
      imageSize=<0 0>
      margin=<ltrb 0 0 0 0>
      outlineSettings {
       bUseBrushTransparency=false
       roundingType=HalfHeightRadius
       width=0
       color {colorUseRule=UseColor_Specified, specifiedColor=<rgba 0 0 0 0>}
       cornerRadii=<0 0 0 0>
      }
      tintColor {colorUseRule=UseColor_Specified, specifiedColor=<rgba 1 1 1 1>}
      uVRegion {bIsValid=false, max=<0 0>, min=<0 0>}
     }
     backgroundImageSelected {
      bIsDynamicallyLoaded=false
      drawAs=Image
      imageType=NoImage
      mirroring=NoMirror
      resourceName=None
      resourceObject=null
      tiling=NoTile
      imageSize=<0 0>
      margin=<ltrb 0 0 0 0>
      outlineSettings {
       bUseBrushTransparency=false
       roundingType=HalfHeightRadius
       width=0
       color {colorUseRule=UseColor_Specified, specifiedColor=<rgba 0 0 0 0>}
       cornerRadii=<0 0 0 0>
      }
      tintColor {colorUseRule=UseColor_Specified, specifiedColor=<rgba 0 0.162 0.7454 1>}
      uVRegion {bIsValid=false, max=<0 0>, min=<0 0>}
     }
     caretImage {
      bIsDynamicallyLoaded=false
      drawAs=Image
      imageType=NoImage
      mirroring=NoMirror
      resourceName=None
      resourceObject=null
      tiling=NoTile
      imageSize=<0 0>
      margin=<ltrb 0 0 0 0>
      outlineSettings {
       bUseBrushTransparency=false
       roundingType=HalfHeightRadius
       width=0
       color {colorUseRule=UseColor_Specified, specifiedColor=<rgba 0 0 0 0>}
       cornerRadii=<0 0 0 0>
      }
      tintColor {colorUseRule=UseColor_Specified, specifiedColor=<rgba 1 1 1 1>}
      uVRegion {bIsValid=false, max=<0 0>, min=<0 0>}
     }
     colorAndOpacity {colorUseRule=UseColor_Foreground, specifiedColor=<rgba 1 0 1 1>}
     font {
      bForceMonospaced=false
      bMaterialIsStencil=false
      fontMaterial=null
      fontObject=&Roboto
      letterSpacing=0
      monospacedWidth=1
      size=12
      skewAmount=0
      typefaceFontName=Bold
      outlineSettings {
       bApplyOutlineToDropShadows=false
       bMiteredCorners=false
       bSeparateFillAlpha=false
       outlineMaterial=null
       outlineSize=0
       outlineColor=<rgba 0 0 0 1>
      }
     }
     hintColorAndOpacity {colorUseRule=UseColor_Foreground, specifiedColor=<rgba 1 0 1 1>}
     selectedBackgroundColor {colorUseRule=UseColor_Foreground, specifiedColor=<rgba 1 0 1 1>}
    }
   }
   slot {class=VerticalBoxSlot, properties {padding=<ltrb -1 0 0 0>}}
  }
  TextBlock_71 {
   widgetClassPath=&TextBlock
   parent=VerticalBox_0
   widget=TextBlock_71
   properties {
    text=Sound Sequence Asset
    colorAndOpacity {colorUseRule=UseColor_Specified, specifiedColor=<rgba 0.5029 0.5029 0.5029 1>}
    font {
     bForceMonospaced=false
     bMaterialIsStencil=false
     fontMaterial=null
     fontObject=&Roboto
     letterSpacing=0
     monospacedWidth=1
     size=12
     skewAmount=0
     typefaceFontName=Bold
     outlineSettings {
      bApplyOutlineToDropShadows=false
      bMiteredCorners=false
      bSeparateFillAlpha=false
      outlineMaterial=null
      outlineSize=0
      outlineColor=<rgba 0 0 0 1>
     }
    }
   }
   slot {class=VerticalBoxSlot}
  }
  HorizontalBox_77 {
   widgetClassPath=&HorizontalBox
   parent=VerticalBox_0
   widget=HorizontalBox_77
   slot {class=VerticalBoxSlot, properties {size {sizeRule=Fill, value=1}}}
  }
  SpokenTextBlock {
   widgetClassPath=&TextBlock
   parent=HorizontalBox_77
   widget=SpokenTextBlock
   properties {
    autoWrapText=true
    text=Sound Sequence
    colorAndOpacity {colorUseRule=UseColor_Specified, specifiedColor=<rgba 0.5029 0.5029 0.5029 1>}
    font {
     bForceMonospaced=false
     bMaterialIsStencil=false
     fontMaterial=null
     fontObject=&Roboto
     letterSpacing=0
     monospacedWidth=1
     size=10.5
     skewAmount=0
     typefaceFontName=Light
     outlineSettings {
      bApplyOutlineToDropShadows=false
      bMiteredCorners=false
      bSeparateFillAlpha=false
      outlineMaterial=null
      outlineSize=0
      outlineColor=<rgba 0 0 0 1>
     }
    }
   }
   slot {class=HorizontalBoxSlot, properties {size {sizeRule=Fill, value=1}}}
  }
  VerticalBox_3 {
   widgetClassPath=&VerticalBox
   parent=HorizontalBox_75
   widget=VerticalBox_3
   slot {class=HorizontalBoxSlot, properties {verticalAlignment=VAlign_Center, padding=<ltrb 4 8 8 8>, size {sizeRule=Fill, value=1}}}
  }
  HorizontalBox_1 {widgetClassPath=&HorizontalBox, parent=VerticalBox_3, widget=HorizontalBox_1, slot {class=VerticalBoxSlot}}
  DialogueWrapperComboBox {
   widgetClassPath=/Script/Blutility.EditorUtilityComboBoxString
   bIsVariable=true
   parent=HorizontalBox_1
   widget=DialogueWrapperComboBox
   properties {
    contentPadding=<ltrb 2 1 2 1>
    font {
     bForceMonospaced=false
     bMaterialIsStencil=false
     fontMaterial=null
     fontObject=&Roboto
     letterSpacing=0
     monospacedWidth=1
     size=9
     skewAmount=0
     typefaceFontName=Regular
     outlineSettings {
      bApplyOutlineToDropShadows=false
      bMiteredCorners=false
      bSeparateFillAlpha=false
      outlineMaterial=null
      outlineSize=0
      outlineColor=<rgba 0 0 0 1>
     }
    }
    foregroundColor {colorUseRule=UseColor_Specified, specifiedColor=<rgba 0.5271 0.5271 0.5271 1>}
   }
   slot {class=HorizontalBoxSlot, properties {size {sizeRule=Fill, value=1}}}
  }
  BrowseToDataAssetButton {
   widgetClassPath=&EditorUtilityButton
   parent=HorizontalBox_1
   widget=BrowseToDataAssetButton
   properties {
    toolTipText=Browse to data asset
    widgetStyle {
     clickedSlateSound {resourceObject=null}
     disabled {
      bIsDynamicallyLoaded=false
      drawAs=RoundedBox
      imageType=NoImage
      mirroring=NoMirror
      resourceName=None
      resourceObject=null
      tiling=NoTile
      imageSize=<0 0>
      margin=<ltrb 0 0 0 0>
      outlineSettings {
       bUseBrushTransparency=false
       roundingType=FixedRadius
       width=1
       color {colorUseRule=UseColor_ColorTable, specifiedColor=<rgba 0 0 0 0>}
       cornerRadii=<4 4 4 4>
      }
      tintColor {colorUseRule=UseColor_Specified, specifiedColor=<rgba 0.0395 0.0395 0.0395 1>}
      uVRegion {bIsValid=false, max=<0 0>, min=<0 0>}
     }
     disabledForeground {colorUseRule=UseColor_ColorTable, specifiedColor=<rgba 0 0 0 0>}
     hovered {
      bIsDynamicallyLoaded=false
      drawAs=RoundedBox
      imageType=NoImage
      mirroring=NoMirror
      resourceName=None
      resourceObject=null
      tiling=NoTile
      imageSize=<32 32>
      margin=<ltrb 0 0 0 0>
      outlineSettings {
       bUseBrushTransparency=false
       roundingType=FixedRadius
       width=1
       color {colorUseRule=UseColor_ColorTable, specifiedColor=<rgba 0 0 0 0>}
       cornerRadii=<4 4 4 4>
      }
      tintColor {colorUseRule=UseColor_Specified, specifiedColor=<rgba 0.0953 0.0953 0.0953 1>}
      uVRegion {bIsValid=false, max=<0 0>, min=<0 0>}
     }
     hoveredForeground {colorUseRule=UseColor_ColorTable, specifiedColor=<rgba 0 0 0 0>}
     hoveredSlateSound {resourceObject=null}
     normal {
      bIsDynamicallyLoaded=false
      drawAs=RoundedBox
      imageType=NoImage
      mirroring=NoMirror
      resourceName=None
      resourceObject=null
      tiling=NoTile
      imageSize=<32 32>
      margin=<ltrb 0 0 0 0>
      outlineSettings {
       bUseBrushTransparency=false
       roundingType=FixedRadius
       width=1
       color {colorUseRule=UseColor_ColorTable, specifiedColor=<rgba 0 0 0 0>}
       cornerRadii=<4 4 4 4>
      }
      tintColor {colorUseRule=UseColor_Specified, specifiedColor=<rgba 0.0395 0.0395 0.0395 1>}
      uVRegion {bIsValid=false, max=<0 0>, min=<0 0>}
     }
     normalForeground {colorUseRule=UseColor_ColorTable, specifiedColor=<rgba 0 0 0 0>}
     normalPadding=<ltrb 0 0 0 0>
     pressed {
      bIsDynamicallyLoaded=false
      drawAs=RoundedBox
      imageType=NoImage
      mirroring=NoMirror
      resourceName=None
      resourceObject=null
      tiling=NoTile
      imageSize=<32 32>
      margin=<ltrb 0 0 0 0>
      outlineSettings {
       bUseBrushTransparency=false
       roundingType=FixedRadius
       width=1
       color {colorUseRule=UseColor_ColorTable, specifiedColor=<rgba 0 0 0 0>}
       cornerRadii=<4 4 4 4>
      }
      tintColor {colorUseRule=UseColor_Specified, specifiedColor=<rgba 0.0284 0.0284 0.0284 1>}
      uVRegion {bIsValid=false, max=<0 0>, min=<0 0>}
     }
     pressedForeground {colorUseRule=UseColor_ColorTable, specifiedColor=<rgba 0 0 0 0>}
     pressedPadding=<ltrb 0 0 0 0>
     pressedSlateSound {resourceObject=null}
    }
   }
   slot {class=HorizontalBoxSlot, properties {padding=<ltrb 4 0 0 0>}}
  }
  PlayIconImage_1 {
   widgetClassPath=&Image
   parent=BrowseToDataAssetButton
   widget=PlayIconImage_1
   properties {
    brush {
     bIsDynamicallyLoaded=false
     drawAs=Image
     imageType=FullColor
     mirroring=NoMirror
     resourceName=None
     resourceObject=/FortUILibrary/Textures/MaterialInstances/MI_UIKit_Icon_SettingsMenu
     tiling=NoTile
     imageSize=<16 16>
     margin=<ltrb 0 0 0 0>
     outlineSettings {
      bUseBrushTransparency=false
      roundingType=HalfHeightRadius
      width=0
      color {colorUseRule=UseColor_Specified, specifiedColor=<rgba 0 0 0 0>}
      cornerRadii=<0 0 0 0>
     }
     tintColor {colorUseRule=UseColor_Specified, specifiedColor=<rgba 1 1 1 1>}
     uVRegion {bIsValid=false, max=<0 0>, min=<0 0>}
    }
    colorAndOpacity=<rgba 0.5029 0.5029 0.5029 1>
   }
   slot {class=ButtonSlot, properties {padding=<ltrb 4 4 4 4>}}
  }
  ReloadBumperData {
   widgetClassPath=&EditorUtilityButton
   parent=HorizontalBox_1
   widget=ReloadBumperData
   properties {
    toolTipText=Refresh list from data asset
    widgetStyle {
     clickedSlateSound {resourceObject=null}
     disabled {
      bIsDynamicallyLoaded=false
      drawAs=RoundedBox
      imageType=NoImage
      mirroring=NoMirror
      resourceName=None
      resourceObject=null
      tiling=NoTile
      imageSize=<0 0>
      margin=<ltrb 0 0 0 0>
      outlineSettings {
       bUseBrushTransparency=false
       roundingType=FixedRadius
       width=1
       color {colorUseRule=UseColor_ColorTable, specifiedColor=<rgba 0 0 0 0>}
       cornerRadii=<4 4 4 4>
      }
      tintColor {colorUseRule=UseColor_Specified, specifiedColor=<rgba 0.0395 0.0395 0.0395 1>}
      uVRegion {bIsValid=false, max=<0 0>, min=<0 0>}
     }
     disabledForeground {colorUseRule=UseColor_ColorTable, specifiedColor=<rgba 0 0 0 0>}
     hovered {
      bIsDynamicallyLoaded=false
      drawAs=RoundedBox
      imageType=NoImage
      mirroring=NoMirror
      resourceName=None
      resourceObject=null
      tiling=NoTile
      imageSize=<32 32>
      margin=<ltrb 0 0 0 0>
      outlineSettings {
       bUseBrushTransparency=false
       roundingType=FixedRadius
       width=1
       color {colorUseRule=UseColor_ColorTable, specifiedColor=<rgba 0 0 0 0>}
       cornerRadii=<4 4 4 4>
      }
      tintColor {colorUseRule=UseColor_Specified, specifiedColor=<rgba 0.0953 0.0953 0.0953 1>}
      uVRegion {bIsValid=false, max=<0 0>, min=<0 0>}
     }
     hoveredForeground {colorUseRule=UseColor_ColorTable, specifiedColor=<rgba 0 0 0 0>}
     hoveredSlateSound {resourceObject=null}
     normal {
      bIsDynamicallyLoaded=false
      drawAs=RoundedBox
      imageType=NoImage
      mirroring=NoMirror
      resourceName=None
      resourceObject=null
      tiling=NoTile
      imageSize=<32 32>
      margin=<ltrb 0 0 0 0>
      outlineSettings {
       bUseBrushTransparency=false
       roundingType=FixedRadius
       width=1
       color {colorUseRule=UseColor_ColorTable, specifiedColor=<rgba 0 0 0 0>}
       cornerRadii=<4 4 4 4>
      }
      tintColor {colorUseRule=UseColor_Specified, specifiedColor=<rgba 0.0395 0.0395 0.0395 1>}
      uVRegion {bIsValid=false, max=<0 0>, min=<0 0>}
     }
     normalForeground {colorUseRule=UseColor_ColorTable, specifiedColor=<rgba 0 0 0 0>}
     normalPadding=<ltrb 0 0 0 0>
     pressed {
      bIsDynamicallyLoaded=false
      drawAs=RoundedBox
      imageType=NoImage
      mirroring=NoMirror
      resourceName=None
      resourceObject=null
      tiling=NoTile
      imageSize=<32 32>
      margin=<ltrb 0 0 0 0>
      outlineSettings {
       bUseBrushTransparency=false
       roundingType=FixedRadius
       width=1
       color {colorUseRule=UseColor_ColorTable, specifiedColor=<rgba 0 0 0 0>}
       cornerRadii=<4 4 4 4>
      }
      tintColor {colorUseRule=UseColor_Specified, specifiedColor=<rgba 0.0284 0.0284 0.0284 1>}
      uVRegion {bIsValid=false, max=<0 0>, min=<0 0>}
     }
     pressedForeground {colorUseRule=UseColor_ColorTable, specifiedColor=<rgba 0 0 0 0>}
     pressedPadding=<ltrb 0 0 0 0>
     pressedSlateSound {resourceObject=null}
    }
   }
   slot {class=HorizontalBoxSlot, properties {padding=<ltrb 4 0 0 0>}}
  }
  PlayIconImage_2 {
   widgetClassPath=&Image
   parent=ReloadBumperData
   widget=PlayIconImage_2
   properties {
    brush {
     bIsDynamicallyLoaded=false
     drawAs=Image
     imageType=FullColor
     mirroring=NoMirror
     resourceName=None
     resourceObject=/Game/Athena/HUD/Inventory/Art/M_UI_Inventory_Loadout_RechargeableIcon
     tiling=NoTile
     imageSize=<16 16>
     margin=<ltrb 0 0 0 0>
     outlineSettings {
      bUseBrushTransparency=false
      roundingType=HalfHeightRadius
      width=0
      color {colorUseRule=UseColor_Specified, specifiedColor=<rgba 0 0 0 0>}
      cornerRadii=<0 0 0 0>
     }
     tintColor {colorUseRule=UseColor_Specified, specifiedColor=<rgba 1 1 1 1>}
     uVRegion {bIsValid=false, max=<0 0>, min=<0 0>}
    }
    colorAndOpacity=<rgba 0.5029 0.5029 0.5029 1>
   }
   slot {class=ButtonSlot, properties {padding=<ltrb 4 4 4 4>}}
  }
  VerticalBox_1 {
   widgetClassPath=&VerticalBox
   parent=HorizontalBox_75
   widget=VerticalBox_1
   slot {class=HorizontalBoxSlot, properties {horizontalAlignment=HAlign_Right, verticalAlignment=VAlign_Center, padding=<ltrb 4 8 8 8>}}
  }
  WrapBox_0 {
   widgetClassPath=&WrapBox
   parent=VerticalBox_1
   widget=WrapBox_0
   properties {innerSlotPadding=<4 4>}
   slot {class=VerticalBoxSlot, properties {size {sizeRule=Fill, value=1}}}
  }
  PlayButton {
   widgetClassPath=&EditorUtilityButton
   parent=WrapBox_0
   widget=PlayButton
   properties {widgetStyle {
     clickedSlateSound {resourceObject=null}
     disabled {
      bIsDynamicallyLoaded=false
      drawAs=RoundedBox
      imageType=NoImage
      mirroring=NoMirror
      resourceName=None
      resourceObject=null
      tiling=NoTile
      imageSize=<0 0>
      margin=<ltrb 0 0 0 0>
      outlineSettings {
       bUseBrushTransparency=false
       roundingType=FixedRadius
       width=1
       color {colorUseRule=UseColor_ColorTable, specifiedColor=<rgba 0 0 0 0>}
       cornerRadii=<4 4 4 4>
      }
      tintColor {colorUseRule=UseColor_Specified, specifiedColor=<rgba 0.0395 0.0395 0.0395 1>}
      uVRegion {bIsValid=false, max=<0 0>, min=<0 0>}
     }
     disabledForeground {colorUseRule=UseColor_ColorTable, specifiedColor=<rgba 0 0 0 0>}
     hovered {
      bIsDynamicallyLoaded=false
      drawAs=RoundedBox
      imageType=NoImage
      mirroring=NoMirror
      resourceName=None
      resourceObject=null
      tiling=NoTile
      imageSize=<32 32>
      margin=<ltrb 0 0 0 0>
      outlineSettings {
       bUseBrushTransparency=false
       roundingType=FixedRadius
       width=1
       color {colorUseRule=UseColor_ColorTable, specifiedColor=<rgba 0 0 0 0>}
       cornerRadii=<4 4 4 4>
      }
      tintColor {colorUseRule=UseColor_Specified, specifiedColor=<rgba 0.0953 0.0953 0.0953 1>}
      uVRegion {bIsValid=false, max=<0 0>, min=<0 0>}
     }
     hoveredForeground {colorUseRule=UseColor_ColorTable, specifiedColor=<rgba 0 0 0 0>}
     hoveredSlateSound {resourceObject=null}
     normal {
      bIsDynamicallyLoaded=false
      drawAs=RoundedBox
      imageType=NoImage
      mirroring=NoMirror
      resourceName=None
      resourceObject=null
      tiling=NoTile
      imageSize=<32 32>
      margin=<ltrb 0 0 0 0>
      outlineSettings {
       bUseBrushTransparency=false
       roundingType=FixedRadius
       width=1
       color {colorUseRule=UseColor_ColorTable, specifiedColor=<rgba 0 0 0 0>}
       cornerRadii=<4 4 4 4>
      }
      tintColor {colorUseRule=UseColor_Specified, specifiedColor=<rgba 0.0395 0.0395 0.0395 1>}
      uVRegion {bIsValid=false, max=<0 0>, min=<0 0>}
     }
     normalForeground {colorUseRule=UseColor_ColorTable, specifiedColor=<rgba 0 0 0 0>}
     normalPadding=<ltrb 0 0 0 0>
     pressed {
      bIsDynamicallyLoaded=false
      drawAs=RoundedBox
      imageType=NoImage
      mirroring=NoMirror
      resourceName=None
      resourceObject=null
      tiling=NoTile
      imageSize=<32 32>
      margin=<ltrb 0 0 0 0>
      outlineSettings {
       bUseBrushTransparency=false
       roundingType=FixedRadius
       width=1
       color {colorUseRule=UseColor_ColorTable, specifiedColor=<rgba 0 0 0 0>}
       cornerRadii=<4 4 4 4>
      }
      tintColor {colorUseRule=UseColor_Specified, specifiedColor=<rgba 0.0284 0.0284 0.0284 1>}
      uVRegion {bIsValid=false, max=<0 0>, min=<0 0>}
     }
     pressedForeground {colorUseRule=UseColor_ColorTable, specifiedColor=<rgba 0 0 0 0>}
     pressedPadding=<ltrb 0 0 0 0>
     pressedSlateSound {resourceObject=null}
    }}
   slot {class=WrapBoxSlot, properties {horizontalAlignment=HAlign_Center, verticalAlignment=VAlign_Center}}
   bIsVariable=~
  }
  PlayIconImage {
   widgetClassPath=&Image
   parent=PlayButton
   widget=PlayIconImage
   properties {
    brush {
     bIsDynamicallyLoaded=false
     drawAs=Image
     imageType=FullColor
     mirroring=NoMirror
     resourceName=None
     resourceObject=/Game/Athena/UI/InGame/Creative/PropertyEditor/Icon_Picker/Textures_V2/64x64/T_UI_IconLibrary_Play_64
     tiling=NoTile
     imageSize=<20 20>
     margin=<ltrb 0 0 0 0>
     outlineSettings {
      bUseBrushTransparency=false
      roundingType=HalfHeightRadius
      width=0
      color {colorUseRule=UseColor_Specified, specifiedColor=<rgba 0 0 0 0>}
      cornerRadii=<0 0 0 0>
     }
     tintColor {colorUseRule=UseColor_Specified, specifiedColor=<rgba 1 1 1 1>}
     uVRegion {bIsValid=false, max=<0 0>, min=<0 0>}
    }
    colorAndOpacity=<rgba 0.5029 0.5029 0.5029 1>
   }
   slot {class=ButtonSlot, properties {padding=<ltrb 4 4 4 4>}}
  }
  StopButton {
   widgetClassPath=&EditorUtilityButton
   parent=WrapBox_0
   widget=StopButton
   properties {widgetStyle {
     clickedSlateSound {resourceObject=null}
     disabled {
      bIsDynamicallyLoaded=false
      drawAs=RoundedBox
      imageType=NoImage
      mirroring=NoMirror
      resourceName=None
      resourceObject=null
      tiling=NoTile
      imageSize=<0 0>
      margin=<ltrb 0 0 0 0>
      outlineSettings {
       bUseBrushTransparency=false
       roundingType=FixedRadius
       width=1
       color {colorUseRule=UseColor_ColorTable, specifiedColor=<rgba 0 0 0 0>}
       cornerRadii=<4 4 4 4>
      }
      tintColor {colorUseRule=UseColor_Specified, specifiedColor=<rgba 0.0395 0.0395 0.0395 1>}
      uVRegion {bIsValid=false, max=<0 0>, min=<0 0>}
     }
     disabledForeground {colorUseRule=UseColor_ColorTable, specifiedColor=<rgba 0 0 0 0>}
     hovered {
      bIsDynamicallyLoaded=false
      drawAs=RoundedBox
      imageType=NoImage
      mirroring=NoMirror
      resourceName=None
      resourceObject=null
      tiling=NoTile
      imageSize=<32 32>
      margin=<ltrb 0 0 0 0>
      outlineSettings {
       bUseBrushTransparency=false
       roundingType=FixedRadius
       width=1
       color {colorUseRule=UseColor_ColorTable, specifiedColor=<rgba 0 0 0 0>}
       cornerRadii=<4 4 4 4>
      }
      tintColor {colorUseRule=UseColor_Specified, specifiedColor=<rgba 0.0953 0.0953 0.0953 1>}
      uVRegion {bIsValid=false, max=<0 0>, min=<0 0>}
     }
     hoveredForeground {colorUseRule=UseColor_ColorTable, specifiedColor=<rgba 0 0 0 0>}
     hoveredSlateSound {resourceObject=null}
     normal {
      bIsDynamicallyLoaded=false
      drawAs=RoundedBox
      imageType=NoImage
      mirroring=NoMirror
      resourceName=None
      resourceObject=null
      tiling=NoTile
      imageSize=<32 32>
      margin=<ltrb 0 0 0 0>
      outlineSettings {
       bUseBrushTransparency=false
       roundingType=FixedRadius
       width=1
       color {colorUseRule=UseColor_ColorTable, specifiedColor=<rgba 0 0 0 0>}
       cornerRadii=<4 4 4 4>
      }
      tintColor {colorUseRule=UseColor_Specified, specifiedColor=<rgba 0.0395 0.0395 0.0395 1>}
      uVRegion {bIsValid=false, max=<0 0>, min=<0 0>}
     }
     normalForeground {colorUseRule=UseColor_ColorTable, specifiedColor=<rgba 0 0 0 0>}
     normalPadding=<ltrb 0 0 0 0>
     pressed {
      bIsDynamicallyLoaded=false
      drawAs=RoundedBox
      imageType=NoImage
      mirroring=NoMirror
      resourceName=None
      resourceObject=null
      tiling=NoTile
      imageSize=<32 32>
      margin=<ltrb 0 0 0 0>
      outlineSettings {
       bUseBrushTransparency=false
       roundingType=FixedRadius
       width=1
       color {colorUseRule=UseColor_ColorTable, specifiedColor=<rgba 0 0 0 0>}
       cornerRadii=<4 4 4 4>
      }
      tintColor {colorUseRule=UseColor_Specified, specifiedColor=<rgba 0.0284 0.0284 0.0284 1>}
      uVRegion {bIsValid=false, max=<0 0>, min=<0 0>}
     }
     pressedForeground {colorUseRule=UseColor_ColorTable, specifiedColor=<rgba 0 0 0 0>}
     pressedPadding=<ltrb 0 0 0 0>
     pressedSlateSound {resourceObject=null}
    }}
   slot {class=WrapBoxSlot, properties {horizontalAlignment=HAlign_Center, verticalAlignment=VAlign_Center}}
   bIsVariable=~
  }
  StopIconImage {
   widgetClassPath=&Image
   parent=StopButton
   widget=StopIconImage
   properties {
    brush {
     bIsDynamicallyLoaded=false
     drawAs=Image
     imageType=FullColor
     mirroring=NoMirror
     resourceName=None
     resourceObject=null
     tiling=NoTile
     imageSize=<20 20>
     margin=<ltrb 0 0 0 0>
     outlineSettings {
      bUseBrushTransparency=false
      roundingType=HalfHeightRadius
      width=0
      color {colorUseRule=UseColor_Specified, specifiedColor=<rgba 0 0 0 0>}
      cornerRadii=<0 0 0 0>
     }
     tintColor {colorUseRule=UseColor_Specified, specifiedColor=<rgba 1 1 1 1>}
     uVRegion {bIsValid=false, max=<0 0>, min=<0 0>}
    }
    colorAndOpacity=<rgba 0.5029 0.5029 0.5029 1>
   }
   slot {class=ButtonSlot, properties {padding=<ltrb 4 4 4 4>}}
  }
  ListViewBorder {
   widgetClassPath=&Border
   parent=VerticalBox_19
   widget=ListViewBorder
   properties {
    visibility=Collapsed
    background {
     bIsDynamicallyLoaded=false
     drawAs=Image
     imageType=NoImage
     mirroring=NoMirror
     resourceName=None
     resourceObject=null
     tiling=NoTile
     imageSize=<32 32>
     margin=<ltrb 0 0 0 0>
     outlineSettings {
      bUseBrushTransparency=false
      roundingType=HalfHeightRadius
      width=0
      color {colorUseRule=UseColor_Specified, specifiedColor=<rgba 0 0 0 0>}
      cornerRadii=<0 0 0 0>
     }
     tintColor {colorUseRule=UseColor_Specified, specifiedColor=<rgba 0.006 0.006 0.006 1>}
     uVRegion {bIsValid=false, max=<0 0>, min=<0 0>}
    }
    padding=<ltrb 0 0 0 0>
   }
   slot {class=VerticalBoxSlot, properties {size {sizeRule=Fill, value=1}}}
  }
  VerticalBox_2 {
   widgetClassPath=&VerticalBox
   parent=ListViewBorder
   widget=VerticalBox_2
   slot {class=BorderSlot, properties {padding=<ltrb 0 0 0 0>}}
  }
  HorizontalBox_0 {widgetClassPath=&HorizontalBox, parent=VerticalBox_2, widget=HorizontalBox_0, slot {class=VerticalBoxSlot}}
  DialogueListView {
   widgetClassPath=/Script/Blutility.EditorUtilityListView
   bIsVariable=true
   parent=HorizontalBox_0
   widget=DialogueListView
   properties {
    bEnableScrollAnimation=true
    entryWidgetClass=&/FortSoundSequenceDialogueEntry.FortSoundSequenceDialogueEntry_C
   }
   slot {class=HorizontalBoxSlot, properties {size {sizeRule=Fill, value=1}}}
  }
  StatusMessage {
   widgetClassPath=&VerticalBox
   parent=VerticalBox_19
   widget=StatusMessage
   properties {visibility=Collapsed}
   slot {class=VerticalBoxSlot, properties {size {sizeRule=Fill, value=1}}}
  }
  HorizontalBox_2 {
   widgetClassPath=&HorizontalBox
   parent=StatusMessage
   widget=HorizontalBox_2
   slot {class=VerticalBoxSlot, properties {size {sizeRule=Fill, value=1}}}
  }
  Border_2 {
   widgetClassPath=&Border
   parent=HorizontalBox_2
   widget=Border_2
   properties {horizontalAlignment=HAlign_Center, verticalAlignment=VAlign_Center, brushColor=<rgba 0.0103 0.0103 0.0103 1>, padding=<ltrb 0 0 0 0>}
   slot {class=HorizontalBoxSlot, properties {size {sizeRule=Fill, value=1}}}
  }
  SizeBox_3 {
   widgetClassPath=&SizeBox
   parent=Border_2
   widget=SizeBox_3
   properties {bOverride_WidthOverride=true, widthOverride=400}
   slot {class=BorderSlot, properties {horizontalAlignment=HAlign_Center, verticalAlignment=VAlign_Center, padding=<ltrb 0 0 0 0>}}
  }
  TextBlock_229 {
   widgetClassPath=&TextBlock
   parent=SizeBox_3
   widget=TextBlock_229
   properties {
    autoWrapText=true
    justification=Center
    text=No sound sequence asset selected.
    colorAndOpacity {colorUseRule=UseColor_Specified, specifiedColor=<rgba 0.5029 0.5029 0.5029 1>}
    font {
     bForceMonospaced=false
     bMaterialIsStencil=false
     fontMaterial=null
     fontObject=&Roboto
     letterSpacing=0
     monospacedWidth=1
     size=9
     skewAmount=0
     typefaceFontName=Regular
     outlineSettings {
      bApplyOutlineToDropShadows=false
      bMiteredCorners=false
      bSeparateFillAlpha=false
      outlineMaterial=null
      outlineSize=0
      outlineColor=<rgba 0 0 0 1>
     }
    }
   }
   slot {class=SizeBoxSlot, properties {verticalAlignment=VAlign_Center}}
  }
 ]
}
