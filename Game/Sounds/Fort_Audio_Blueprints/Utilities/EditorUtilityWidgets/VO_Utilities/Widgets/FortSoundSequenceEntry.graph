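# Editor utility list-row widget for a VO utilities tool, representing one FortSoundSequenceData item or the "Unassigned Lines" bucket when its asset name is empty.
# It caches the list item and selection state, which getter functions use for name, icon, alternating row colors and delete visibility.
# Dropping a FortDialogueData payload disables its unassign button and broadcasts On Reassign Dialogue Data with this sequence as target.
# The delete button only broadcasts On Delete Sound Sequence Clicked.
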EditorUtilityWidgetBlueprint /Game/Sounds/Fort_Audio_Blueprints/Utilities/EditorUtilityWidgets/VO_Utilities/Widgets/FortSoundSequenceEntry
graphs [
 Delete is Enabled {nodes [
   Delete is Enabled() {is=title K2Node_FunctionEntry}
   > n1 = Get!(execute:Exec) -> (then:Exec, else:Exec, Sound Sequence Data:Fort Sound Sequence Data Object Reference)
   > Return Node!(
    execute
    ReturnValue:ESlateVisibility Enum=Select(Option 0=Collapsed, Option 1=Visible, Index=Get Enable Delete!(self:Fort Sound Sequence Data Object Reference=@n1) -> (Enable Delete:bool))
   )
  ]}
 EventGraph {nodes [
   "On Clicked (DeleteSoundSequenceButton)"() {is=title K2Node_ComponentBoundEvent}
   > Call On Delete Sound Sequence Clicked!(
    execute:Exec
    self:Fort Sound Sequence Entry Object Reference
    Sound Sequence Data:Fort Sound Sequence Data Object Reference=Get Sound Sequence Data() -> (Sound Sequence Data:Fort Sound Sequence Data Object Reference)
   ) -> (then:Exec) {is=title K2Node_CallDelegate}
   Event OnDragLeave!() -> (OutputDelegate, then, PointerEvent:Pointer Event Structure, Operation:Drag Drop Operation Object Reference) {is=title K2Node_Event}
   > Set Valid Drag Entered(Valid Drag Entered:bool=false) -> (Output_Get:bool)
   n1 = Event On Item Selection Changed!() -> (OutputDelegate, then, bIsSelected:bool) {is=title K2Node_Event}
   > Set Is Selected(Is Selected:bool=@n1.bIsSelected) -> (Output_Get:bool)
   n2 = Event OnListItemObjectSet()
   > n3 = Cast To FortSoundSequenceData!(execute, Object=@n2.ListItemObject) -> (then, CastFailed, AsFort Sound Sequence Data:Fort Sound Sequence Data Object Reference) {is=title K2Node_DynamicCast}
   > Set Sound Sequence Data(Sound Sequence Data:Fort Sound Sequence Data Object Reference=@n3) -> (Output_Get:Fort Sound Sequence Data Object Reference)
  ]}
 Get Asset Name {nodes [
   Get Asset Name() {is=title K2Node_FunctionEntry}
   n1 = Get Asset Name!(self:Fort Sound Sequence Data Object Reference=Get Sound Sequence Data()) -> (Asset Name:String) {is=title K2Node_VariableGet}
   n2 = String To Name(InString=@n1)
   > Return Node(ReturnValue:Text=SelectText(A="LOCTEXT(\"Unassigned Lines\")", B="To Text (Name)"!(InName:Name=@n2) -> (ReturnValue:Text), bPickA="Equal (Name)"(A:String=@n1)))
  ]}
 Get Background Color {nodes [
   Get Background Color() {is=title K2Node_FunctionEntry}
   n1 = IsHovered()
   n2 = Get Row Index!(self:Fort Sound Sequence Data Object Reference=Get Sound Sequence Data()) -> (Row Index:int=0)
   n3 = Select(
    Option 0:Linear Color Structure=(R=0.0103,G=0.0103,B=0.0103,A=1)
    Option 1:Linear Color Structure=(R=0.0075,G=0.0075,B=0.0075,A=1)
    Index="Equal (Integer)"(B="% (Integer)"(A=@n2, B=2) -> (ReturnValue=0))
   ) -> (ReturnValue:Linear Color Structure)
   > Return Node(ReturnValue:Linear Color Structure=Select(
     Option 0=Select(Option 0=@n3, Option 1=(R=0.0176,G=0.0176,B=0.0176,A=1), Index=@n1)
     Option 1=(R=0,G=0.162,B=0.7454,A=1)
     Index=Get Is Selected() -> (Is Selected:bool)
    ))
  ]}
 Get Color {nodes [
   Get Color() {is=title K2Node_FunctionEntry}
   n1 = Get Valid Drag Entered() -> (Valid Drag Entered:bool)
   n2 = Get Is Selected()
   > Return Node(ReturnValue=Select(Option 0=(R=0.5029,G=0.5029,B=0.5029,A=1), Option 1=(R=1,G=1,B=1,A=1), Index=OR Boolean(A=@n2, B=@n1)))
  ]}
 Get Enable Remove Icon {nodes [
   Get Enable Remove Icon() {is=title K2Node_FunctionEntry}
   n1 = Select(Option 0=(R=0.2502,G=0.2502,B=0.2502,A=1), Option 1=(R=0.8802,G=0.8802,B=0.8802,A=1), Index=Get Is Selected())
   > Return Node(ReturnValue:Slate Brush Structure=Make Slate Brush!(
     TintColor:Slate Color Structure=Make Slate Color(SpecifiedColor=@n1)
     DrawAs:ESlateBrushDrawType Enum=Image
     Tiling:ESlateBrushTileType Enum=NoTile
     Mirroring:ESlateBrushMirrorType Enum=NoMirror
     ImageSize:Vector 2D Structure=MakeVector2D(X=24, Y=24)
     Margin:Margin Structure=(Left=0,Top=0,Right=0,Bottom=0)
     ResourceObject:Object Reference=/Game/UI/Foundation/Textures/Buttons/T-SlotIcon-NotRecyclable-64
     OutlineSettings:Slate Brush Outline Settings Structure=(CornerRadii=(X=0,Y=0,Z=0,W=0),Color=(SpecifiedColor=(R=0,G=0,B=0,A=0),ColorUseRule=UseColor_Specified),Width=0,RoundingType=HalfHeightRadius,bUseBrushTransparency=False)
    ) -> (SlateBrush:Slate Brush Structure) {is=title K2Node_MakeStruct})
  ]}
 GetMainIconBrush {nodes [
   GetMainIconBrush() {is=title K2Node_FunctionEntry}
   n1 = Get Asset Name(self=Get Sound Sequence Data())
   > Return Node(ReturnValue=Select(
     Option 0:Slate Brush Structure=Make Slate Brush(
      TintColor=(SpecifiedColor=(R=1,G=1,B=1,A=1),ColorUseRule=UseColor_Specified)
      DrawAs=Image
      Tiling=NoTile
      Mirroring=NoMirror
      ImageSize=(X=32,Y=32)
      Margin=(Left=0,Top=0,Right=0,Bottom=0)
      ResourceObject=/Game/Athena/UI/InGame/Creative/PropertyEditor/Icon_Picker/Textures_V2/64x64/T_UI_IconLibrary_Chat_64
      OutlineSettings=(CornerRadii=(X=0,Y=0,Z=0,W=0),Color=(SpecifiedColor=(R=0,G=0,B=0,A=0),ColorUseRule=UseColor_Specified),Width=0,RoundingType=HalfHeightRadius,bUseBrushTransparency=False)
     )
     Option 1:Slate Brush Structure=Make Slate Brush(
      TintColor=(SpecifiedColor=(R=1,G=1,B=1,A=1),ColorUseRule=UseColor_Specified)
      DrawAs=Image
      Tiling=NoTile
      Mirroring=NoMirror
      ImageSize=(X=32,Y=32)
      Margin=(Left=0,Top=0,Right=0,Bottom=0)
      ResourceObject=/Game/Athena/UI/InGame/Creative/PropertyEditor/Icon_Picker/Textures/64x64/T_UI_IconLibrary_X_64
      OutlineSettings=(CornerRadii=(X=0,Y=0,Z=0,W=0),Color=(SpecifiedColor=(R=0,G=0,B=0,A=0),ColorUseRule=UseColor_Specified),Width=0,RoundingType=HalfHeightRadius,bUseBrushTransparency=False)
     )
     Index="Equal (Name)"(A=@n1)
    ) -> (ReturnValue:Slate Brush Structure))
  ]}
 Get_DragEnteredGlow_Visibility {nodes [
   Get_DragEnteredGlow_Visibility() {is=title K2Node_FunctionEntry}
   > Return Node(ReturnValue:ESlateVisibility Enum=Select(Option 0:ESlateVisibility Enum=Hidden, Option 1:ESlateVisibility Enum=Visible, Index=Get Valid Drag Entered()) -> (ReturnValue:ESlateVisibility Enum))
  ]}
 On Delete Sound Sequence Clicked {nodes [On Delete Sound Sequence Clicked!() -> (then, Sound Sequence Data:Fort Sound Sequence Data Object Reference) {is=title K2Node_FunctionEntry}]}
 On Reassign Dialogue Data {nodes [On Reassign Dialogue Data!() -> (then, Dialogue To Reassign:Fort Dialogue Data Object Reference, Target Sound Sequence:Fort Sound Sequence Data Object Reference) {is=title K2Node_FunctionEntry}]}
 OnDragOver {nodes [
   n1 = OnDragOver!() -> (then, MyGeometry:Geometry Structure, PointerEvent:Pointer Event Structure, Operation:Drag Drop Operation Object Reference) {is=title K2Node_FunctionEntry}
   > n2 = Cast To FortDialogueData!(execute, Object=Get Payload!(self:Drag Drop Operation Object Reference=@n1.Operation) -> (Payload:Object Reference)) -> (then, CastFailed, AsFort Dialogue Data:Fort Dialogue Data Object Reference) {is=title K2Node_DynamicCast}
   > Set Valid Drag Entered(Valid Drag Entered=true)
   > Return Node(ReturnValue:bool=true)
   Return Node(execute=@n2.CastFailed, ReturnValue=true)
  ]}
 OnDrop {nodes [
   n1 = OnDrop!() -> (then, MyGeometry:Geometry Structure, PointerEvent:Pointer Event Structure, Operation:Drag Drop Operation Object Reference) {is=title K2Node_FunctionEntry}
   > Set Valid Drag Entered(Valid Drag Entered=false)
   > n2 = Cast To FortDialogueData(Object=Get Payload(self=@n1.Operation))
   > Set Enable Unassign Button!(execute, Enable Unassign Button:bool=false, self:Fort Dialogue Data Object Reference=@n2) -> (then, Output_Get:bool)
   > Call On Reassign Dialogue Data!(
    execute:Exec
    self:Fort Sound Sequence Entry Object Reference
    Dialogue To Reassign:Fort Dialogue Data Object Reference=@n2
    Target Sound Sequence:Fort Sound Sequence Data Object Reference=Get Sound Sequence Data()
   ) -> (then:Exec) {is=title K2Node_CallDelegate}
   > Return Node(ReturnValue=true)
   Return Node(execute=@n2.CastFailed, ReturnValue=true)
  ]}
 Reset {nodes [
   Reset() {is=title K2Node_FunctionEntry}
   > Set Sound Sequence Data()
   > Set Is Selected(Is Selected=false)
   > Set Valid Drag Entered(Valid Drag Entered=false)
  ]}
 Visible If Unassigned Lines {nodes [
   Visible If Unassigned Lines() {is=title K2Node_FunctionEntry}
   n1 = Get Asset Name(self=Get Sound Sequence Data())
   > Return Node(ReturnValue:ESlateVisibility Enum=Select(Option 0=Collapsed, Option 1=Visible, Index="Equal (Name)"(A=@n1)))
  ]}
 Visible If Valid Asset Name {nodes [
   Visible If Valid Asset Name() {is=title K2Node_FunctionEntry}
   n1 = Get Asset Name(self=Get Sound Sequence Data())
   > Return Node(ReturnValue=Select(Option 0=Collapsed, Option 1=Visible, Index="Not Equal (Name)"(A:String=@n1) -> (ReturnValue=~)))
  ]}
]
variables [
 Is Selected {type=boolean}
 On Delete Sound Sequence Clicked {default=()}
 On Reassign Dialogue Data {default=()}
 Sound Sequence Data {type=FortSoundSequenceData_C, default=null}
 Valid Drag Entered {type=boolean}
]
widgets {
 info {rootWidgetClass=&VerticalBox, widgetCount=21}
 widgets [
  VerticalBox_21 {widgetClassPath=&VerticalBox, widget=VerticalBox_21}
  HorizontalBox_35 {widgetClassPath=&HorizontalBox, parent=VerticalBox_21, widget=HorizontalBox_35, slot {class=VerticalBoxSlot}}
  Overlay_0 {
   widgetClassPath=&Overlay
   parent=HorizontalBox_35
   widget=Overlay_0
   slot {class=HorizontalBoxSlot, properties {size {sizeRule=Fill, value=1}}}
  }
  Background {
   widgetClassPath=&Border
   bIsVariable=true
   parent=Overlay_0
   widget=Background
   properties {
    verticalAlignment=VAlign_Center
    background {
     bIsDynamicallyLoaded=false
     drawAs=Box
     imageType=NoImage
     mirroring=NoMirror
     resourceName=None
     resourceObject=null
     tiling=NoTile
     imageSize=<32 32>
     margin=<ltrb 0 0 0 0>
     outlineSettings {
      bUseBrushTransparency=false
      roundingType=FixedRadius
      width=2
      color {colorUseRule=UseColor_Specified, specifiedColor=<rgba 0 0 0 0>}
      cornerRadii=<0 0 0 0>
     }
     tintColor {colorUseRule=UseColor_Specified, specifiedColor=<rgba 1 1 1 1>}
     uVRegion {bIsValid=false, max=<0 0>, min=<0 0>}
    }
    brushColor=<rgba 0.0176 0.0176 0.0176 1>
    padding=<ltrb 6 6 6 4>
   }
   slot {class=OverlaySlot, properties {horizontalAlignment=HAlign_Fill, verticalAlignment=VAlign_Fill}}
  }
  DragEnteredGlow {
   widgetClassPath=&Border
   bIsVariable=true
   parent=Overlay_0
   widget=DragEnteredGlow
   properties {
    visibility=Hidden
    background {
     bIsDynamicallyLoaded=false
     drawAs=Image
     imageType=FullColor
     mirroring=NoMirror
     resourceName=None
     resourceObject=null
     tiling=NoTile
     imageSize=<256 256>
     margin=<ltrb 0 0 0 0>
     outlineSettings {
      bUseBrushTransparency=false
      roundingType=HalfHeightRadius
      width=0
      color {colorUseRule=UseColor_Specified, specifiedColor=<rgba 0 0 0 0>}
      cornerRadii=<0 0 0 0>
     }
     tintColor {colorUseRule=UseColor_Specified, specifiedColor=<rgba 1 1 1 1>}
     uVRegion {bIsValid=false, max=<0 0>, min=<0 0>}
    }
    brushColor=<rgba 0 0.5524 1 0.5>
    padding=<ltrb 0 0 0 0>
   }
   slot {class=OverlaySlot, properties {horizontalAlignment=HAlign_Fill, verticalAlignment=VAlign_Fill}}
  }
  Content {
   widgetClassPath=&HorizontalBox
   parent=Overlay_0
   widget=Content
   slot {class=OverlaySlot, properties {horizontalAlignment=HAlign_Fill, verticalAlignment=VAlign_Center, padding=<ltrb 6 6 6 4>}}
  }
  VerticalBox {widgetClassPath=&VerticalBox, parent=Content, widget=VerticalBox, slot {class=HorizontalBoxSlot}}
  SizeBox_0 {
   widgetClassPath=&SizeBox
   parent=VerticalBox
   widget=SizeBox_0
   properties {bOverride_HeightOverride=true, bOverride_WidthOverride=true, heightOverride=32, widthOverride=32}
   slot {class=VerticalBoxSlot, properties {horizontalAlignment=HAlign_Center, verticalAlignment=VAlign_Center, padding=<ltrb 4 4 4 4>}}
  }
  SoundIcon {
   widgetClassPath=&Image
   bIsVariable=true
   parent=SizeBox_0
   widget=SoundIcon
   properties {
    brush {
     bIsDynamicallyLoaded=false
     drawAs=Image
     imageType=FullColor
     mirroring=NoMirror
     resourceName=None
     resourceObject=&T_UI_IconLibrary_Chat_64
     tiling=NoTile
     imageSize=<64 64>
     margin=<ltrb 0 0 0 0>
     outlineSettings {
      bUseBrushTransparency=false
      roundingType=HalfHeightRadius
      width=0
      color {colorUseRule=UseColor_Specified, specifiedColor=<rgba 0 0 0 0>}
      cornerRadii=<0 0 0 0>
     }
     tintColor {colorUseRule=UseColor_Specified, specifiedColor=<rgba 1 1 1 1>}
     uVRegion {bIsValid=false, max=<0 0>, min=<0 0>}
    }
    colorAndOpacity=<rgba 0.5029 0.5029 0.5029 1>
   }
   slot {class=SizeBoxSlot}
  }
  VerticalBox_23 {
   widgetClassPath=&VerticalBox
   parent=Content
   widget=VerticalBox_23
   slot {class=HorizontalBoxSlot, properties {verticalAlignment=VAlign_Center, padding=<ltrb 6 0 0 0>, size {sizeRule=Fill, value=1}}}
  }
  HorizontalBox_37 {widgetClassPath=&HorizontalBox, parent=VerticalBox_23, widget=HorizontalBox_37, slot {class=VerticalBoxSlot}}
  AssetNameText {
   widgetClassPath=&TextBlock
   parent=HorizontalBox_37
   widget=AssetNameText
   properties {
    autoWrapText=true
    text=Asset Name
    colorAndOpacity {colorUseRule=UseColor_Specified, specifiedColor=<rgba 0.5029 0.5029 0.5029 1>}
    font {
     bForceMonospaced=false
     bMaterialIsStencil=false
     fontMaterial=null
     fontObject=&Roboto
     letterSpacing=0
     monospacedWidth=1
     size=9.75
     skewAmount=0
     typefaceFontName=Regular
     outlineSettings {
      bApplyOutlineToDropShadows=false
      bMiteredCorners=false
      bSeparateFillAlpha=false
      outlineMaterial=null
      outlineSize=0
      outlineColor=<rgba 0 0 0 1>
     }
    }
   }
   slot {class=HorizontalBoxSlot, properties {size {sizeRule=Fill, value=1}}}
  }
  HorizontalBox_2 {
   widgetClassPath=&HorizontalBox
   parent=VerticalBox_23
   widget=HorizontalBox_2
   properties {visibility=Collapsed}
   slot {class=VerticalBoxSlot}
  }
  UnassignedLines {
   widgetClassPath=&TextBlock
   parent=HorizontalBox_2
   widget=UnassignedLines
   properties {
    text=Unassigned Lines
    colorAndOpacity {colorUseRule=UseColor_Specified, specifiedColor=<rgba 0.5029 0.5029 0.5029 1>}
    font {
     bForceMonospaced=false
     bMaterialIsStencil=false
     fontMaterial=null
     fontObject=&Roboto
     letterSpacing=0
     monospacedWidth=1
     size=9
     skewAmount=0
     typefaceFontName=Regular
     outlineSettings {
      bApplyOutlineToDropShadows=false
      bMiteredCorners=false
      bSeparateFillAlpha=false
      outlineMaterial=null
      outlineSize=0
      outlineColor=<rgba 0 0 0 1>
     }
    }
   }
   slot {class=HorizontalBoxSlot}
  }
  HorizontalBox {
   widgetClassPath=&HorizontalBox
   parent=VerticalBox_23
   widget=HorizontalBox
   slot {class=VerticalBoxSlot, properties {size {sizeRule=Fill, value=1}}}
  }
  SpokenTextBlock {
   widgetClassPath=&TextBlock
   bIsVariable=true
   parent=HorizontalBox
   widget=SpokenTextBlock
   properties {
    autoWrapText=true
    text=Sound Sequence
    colorAndOpacity {colorUseRule=UseColor_Specified, specifiedColor=<rgba 0.5029 0.5029 0.5029 1>}
    font {
     bForceMonospaced=false
     bMaterialIsStencil=false
     fontMaterial=null
     fontObject=&Roboto
     letterSpacing=0
     monospacedWidth=1
     size=9
     skewAmount=0
     typefaceFontName=Light
     outlineSettings {
      bApplyOutlineToDropShadows=false
      bMiteredCorners=false
      bSeparateFillAlpha=false
      outlineMaterial=null
      outlineSize=0
      outlineColor=<rgba 0 0 0 1>
     }
    }
   }
   slot {class=HorizontalBoxSlot, properties {size {sizeRule=Fill, value=1}}}
  }
  VerticalBox_1 {
   widgetClassPath=&VerticalBox
   parent=Content
   widget=VerticalBox_1
   slot {class=HorizontalBoxSlot, properties {horizontalAlignment=HAlign_Right, verticalAlignment=VAlign_Center}}
  }
  HorizontalBox_1 {widgetClassPath=&HorizontalBox, parent=VerticalBox_1, widget=HorizontalBox_1, slot {class=VerticalBoxSlot}}
  SizeBox_1 {widgetClassPath=&SizeBox, parent=HorizontalBox_1, widget=SizeBox_1, slot {class=HorizontalBoxSlot}}
  DeleteSoundSequenceButton {
   widgetClassPath=&EditorUtilityButton
   parent=SizeBox_1
   widget=DeleteSoundSequenceButton
   properties {
    toolTipText="Remove Sound Sequence (All dialogue data is moved to Unassigned Lines)"
    widgetStyle {
     clickedSlateSound {resourceObject=null}
     disabled {
      bIsDynamicallyLoaded=false
      drawAs=RoundedBox
      imageType=NoImage
      mirroring=NoMirror
      resourceName=None
      resourceObject=null
      tiling=NoTile
      imageSize=<0 0>
      margin=<ltrb 0 0 0 0>
      outlineSettings {
       bUseBrushTransparency=false
       roundingType=FixedRadius
       width=1
       color {colorUseRule=UseColor_ColorTable, specifiedColor=<rgba 0 0 0 0>}
       cornerRadii=<4 4 4 4>
      }
      tintColor {colorUseRule=UseColor_Specified, specifiedColor=<rgba 0.0395 0.0395 0.0395 1>}
      uVRegion {bIsValid=false, max=<0 0>, min=<0 0>}
     }
     disabledForeground {colorUseRule=UseColor_ColorTable, specifiedColor=<rgba 0 0 0 0>}
     hovered {
      bIsDynamicallyLoaded=false
      drawAs=RoundedBox
      imageType=NoImage
      mirroring=NoMirror
      resourceName=None
      resourceObject=null
      tiling=NoTile
      imageSize=<0 0>
      margin=<ltrb 0 0 0 0>
      outlineSettings {
       bUseBrushTransparency=false
       roundingType=FixedRadius
       width=0
       color {colorUseRule=UseColor_Foreground, specifiedColor=<rgba 0 0 0 0>}
       cornerRadii=<4 4 4 4>
      }
      tintColor {colorUseRule=UseColor_Specified, specifiedColor=<rgba 1 1 1 0.1>}
      uVRegion {bIsValid=false, max=<0 0>, min=<0 0>}
     }
     hoveredForeground {colorUseRule=UseColor_ColorTable, specifiedColor=<rgba 0 0 0 0>}
     hoveredSlateSound {resourceObject=null}
     normal {
      bIsDynamicallyLoaded=false
      drawAs=NoDrawType
      imageType=NoImage
      mirroring=NoMirror
      resourceName=None
      resourceObject=null
      tiling=NoTile
      imageSize=<0 0>
      margin=<ltrb 0 0 0 0>
      outlineSettings {
       bUseBrushTransparency=false
       roundingType=FixedRadius
       width=0
       color {colorUseRule=UseColor_Foreground, specifiedColor=<rgba 0 0 0 0>}
       cornerRadii=<4 4 4 4>
      }
      tintColor {colorUseRule=UseColor_Specified, specifiedColor=<rgba 0.0312 0.0312 0.0312 1>}
      uVRegion {bIsValid=false, max=<0 0>, min=<0 0>}
     }
     normalForeground {colorUseRule=UseColor_ColorTable, specifiedColor=<rgba 0 0 0 0>}
     normalPadding=<ltrb 0 0 0 0>
     pressed {
      bIsDynamicallyLoaded=false
      drawAs=RoundedBox
      imageType=NoImage
      mirroring=NoMirror
      resourceName=None
      resourceObject=null
      tiling=NoTile
      imageSize=<0 0>
      margin=<ltrb 0 0 0 0>
      outlineSettings {
       bUseBrushTransparency=false
       roundingType=FixedRadius
       width=0
       color {colorUseRule=UseColor_Foreground, specifiedColor=<rgba 0 0 0 0>}
       cornerRadii=<4 4 4 4>
      }
      tintColor {colorUseRule=UseColor_Specified, specifiedColor=<rgba 0.0625 0.0625 0.0625 1>}
      uVRegion {bIsValid=false, max=<0 0>, min=<0 0>}
     }
     pressedForeground {colorUseRule=UseColor_ColorTable, specifiedColor=<rgba 0 0 0 0>}
     pressedPadding=<ltrb 0 0 0 0>
     pressedSlateSound {resourceObject=null}
    }
   }
   slot {class=SizeBoxSlot, properties {horizontalAlignment=HAlign_Left, verticalAlignment=VAlign_Top}}
  }
  IconImage {
   widgetClassPath=&Image
   bIsVariable=true
   parent=DeleteSoundSequenceButton
   widget=IconImage
   properties {brush {
     bIsDynamicallyLoaded=false
     drawAs=Image
     imageType=FullColor
     mirroring=NoMirror
     resourceName=None
     resourceObject=&T-SlotIcon-NotRecyclable-64
     tiling=NoTile
     imageSize=<24 24>
     margin=<ltrb 0 0 0 0>
     outlineSettings {
      bUseBrushTransparency=false
      roundingType=HalfHeightRadius
      width=0
      color {colorUseRule=UseColor_Specified, specifiedColor=<rgba 0 0 0 0>}
      cornerRadii=<0 0 0 0>
     }
     tintColor {colorUseRule=UseColor_Specified, specifiedColor=<rgba 1 1 1 1>}
     uVRegion {bIsValid=false, max=<0 0>, min=<0 0>}
    }}
   slot {class=ButtonSlot, properties {padding=<ltrb 2 2 2 2>}}
  }
 ]
}
