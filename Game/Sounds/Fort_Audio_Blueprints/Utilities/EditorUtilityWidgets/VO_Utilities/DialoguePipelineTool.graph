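# Editor utility widget that turns a VO script data table into Sound Sequence assets for dialogue.
# Table rows become dialogue data objects, grouped by Sequence column or held as unassigned, for editing in a list and dialogue viewer.
# An import step brings in sound waves, generates Dialogue Waves and Sound Cues, and applies default audio settings.
# Asset creation runs as an Editor Utility Task after confirming and deleting name-matched existing assets.
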EditorUtilityWidgetBlueprint /Game/Sounds/Fort_Audio_Blueprints/Utilities/EditorUtilityWidgets/VO_Utilities/DialoguePipelineTool
graphs [
 Add New Sound Sequence {nodes [
   Add New Sound Sequence() {is=title K2Node_FunctionEntry}
   n1 = Get EnableSoundSequenceRemovalButton() -> (EnableSoundSequenceRemovalButton:Fort Sound Sequence Editor Toggle Button Object Reference)
   > n2 = Construct Object from Class!(
    execute
    Class=/Game/Sounds/Fort_Audio_Blueprints/Utilities/EditorUtilityWidgets/VO_Utilities/Data/FortSoundSequenceData.FortSoundSequenceData_C
    self
    Row Index:int=Length(TargetArray:Array of Fort Sound Sequence Data Object References=Get Sound Sequence Data Objects() -> (Sound Sequence Data Objects:Array of Fort Sound Sequence Data Object References))
    Asset Name:String=NewSoundSequence
    Dialogue Bumpers:Name=None
    Dialogue Data:Array of Fort Dialogue Data Object References
    Enable Delete:bool=IsChecked(self=Get ToggleButton!(self:Fort Sound Sequence Editor Toggle Button Object Reference=@n1) -> (ToggleButton:Editor Utility Check Box Object Reference))
   ) -> (then, ReturnValue:Fort Sound Sequence Data Object Reference)
   > Add(
    TargetArray:Array of Fort Sound Sequence Data Object References=Get Sound Sequence Data Objects()
    NewItem:"Fort Sound Sequence Data Object Reference (by ref)"=@n2
   ) {is=title K2Node_CallArrayFunction}
   > AddItem!(
    execute
    self:List View Object Reference=Get SoundSequenceList() -> (SoundSequenceList:Editor Utility List View Object Reference)
    Item:Object Reference=@n2
   ) -> (then)
  ]}
 Can Create Assets {nodes [
   Can Create Assets() {is=title K2Node_FunctionEntry}
   > n1 = Sequence()
   > Set Has Sound Sequence Data(Has Sound Sequence Data:bool=Is Not Empty!(TargetArray:Array of Fort Sound Sequence Data Object References=Get Sound Sequence Data Objects()) -> (ReturnValue:bool)) -> (Output_Get:bool)
   n2 = Sequence(execute=@n1.then_1)
   > Clear(TargetArray=Get Unique Names() -> (Unique Names:Array of Strings))
   > Clear(TargetArray=Get Duplicate Names() -> (Duplicate Names:Array of Strings))
   n3 = Sequence(execute=@n2.then_1)
   > n4 = For Each Loop(Array:Array of Fort Sound Sequence Data Object References=Get Sound Sequence Data Objects()) -> (Array Element:Fort Sound Sequence Data Object Reference)
   n5 = Get Asset Name!(self:Fort Sound Sequence Data Object Reference=@n4) -> (Asset Name:String)
   > n6 = Branch(Condition=Contains Item(TargetArray:Array of Strings=Get Unique Names(), ItemToFind:"String (by ref)"=@n5))
   > Add(TargetArray:Array of Strings=Get Duplicate Names(), NewItem:"String (by ref)"=@n5)
   Add(execute=@n6.else, TargetArray=Get Unique Names(), NewItem=@n5)
   Sequence(execute=@n3.then_1)
   n7 = Get Duplicate Names()
   > Return Node!(
    execute
    ReturnValue:bool=AND Boolean(A=Get Has Sound Sequence Data() -> (Has Sound Sequence Data:bool), B=Is Empty!(TargetArray=@n7) -> (ReturnValue:bool))
   )
  ]}
 Create Data Objects from Data Table {nodes [
   Create Data Objects from Data Table() {is=title K2Node_FunctionEntry}
   > n1 = GetDataTableRowNames(Table=Get Script Data Table() -> (Script Data Table:Data Table Object Reference))
   > n2 = For Each Loop(Array:Array of Names=@n1) -> (Array Element:Name)
   n3 = Values!(
    execute=@n2.Completed
    TargetMap:Map of Strings to Fort Sound Sequence Data Object References=Get Local Sound Sequence Objects() -> (Local Sound Sequence Objects:Map of Strings to Fort Sound Sequence Data Object References)
   ) -> (then, Values:Array of Fort Sound Sequence Data Object References)
   > Set Sound Sequence Data Objects(Sound Sequence Data Objects:Array of Fort Sound Sequence Data Object References=@n3) -> (Output_Get:Array of Fort Sound Sequence Data Object References)
   Set Row Name(execute=@n2.LoopBody, Row Name:Name=@n2) -> (Output_Get:Name=None)
   > Set Line Number(Line Number:int=int + int(A=@n2.Array Index)) -> (Output_Get:int=0)
   > n4 = Sequence()
   > n5 = Locate Sound!(
    execute
    self
    Sounds Folder:String=Get Subfolder Path!(Asset:Object Reference=Get Script Data Table(), Subfolder Name:String=SoundCues, __WorldContext:Object Reference) -> (Subfolder Path:String) {is=title K2Node_CallFunction}
    Row Name:Name=Get Row Name() -> (Row Name:Name=None)
   ) -> (then, Sound:Sound Base Object Reference)
   > Set Sound(Sound:Sound Base Object Reference=@n5) -> (Output_Get:Sound Base Object Reference)
   n6 = Sequence(execute=@n4.then_1)
   > n7 = Get Data Table Row!(execute:Exec, DataTable:Data Table Object Reference=Get Script Data Table(), RowName:Name=Get Row Name() -> (Row Name=None)) -> (then:Exec, RowNotFound:Exec, ReturnValue:VO Dialogue Wave Struct Structure)
   n8 = Break VO Dialogue Wave Struct!(VO_DialogueWave_Struct:"VO Dialogue Wave Struct Structure (by ref)"=@n7) -> (
    Character_17_#1:String
    Dialogue_15_#2:String
    Sequence_11_#3:String
    Context_24_#4:String
    Post-SentenceDelay_25_#5:double=~
   ) {is=title K2Node_BreakStruct}
   > Set Character(Character:String=@n8) -> (Output_Get:String)
   > Set Spoken Text(Spoken Text:String=@n8.Dialogue_15_#2) -> (Output_Get:String)
   > Set Sequence(Sequence:String=@n8.Sequence_11_#3) -> (Output_Get:String)
   > Set Context(Context:String=@n8.Context_24_#4) -> (Output_Get:String)
   > Set Post-Sentence Delay(Post-Sentence Delay:double=@n8.Post-SentenceDelay_25_#5) -> (Output_Get:double)
   n9 = Sequence(execute=@n6.then_1)
   n10 = Get Post-Sentence Delay() -> (Post-Sentence Delay:double)
   > n11 = Construct Object from Class!(
    execute
    Class=/Game/Sounds/Fort_Audio_Blueprints/Utilities/EditorUtilityWidgets/VO_Utilities/Data/FortDialogueData.FortDialogueData_C
    self
    Line Number:int=Get Line Number() -> (Line Number:int=0)
    Sequence:String=Get Sequence() -> (Sequence:String)
    Row Name:Name=Get Row Name() -> (Row Name=None)
    Character:String=Get Character() -> (Character:String)
    Spoken Text:Text="To Text (String)"(InString=Get Spoken Text() -> (Spoken Text:String))
    Context:String=Get Context() -> (Context:String)
    Sound:Sound Base Object Reference=Get Sound() -> (Sound:Sound Base Object Reference)
    Post-Sentence Delay:double=Select(Option 0:double=@n10, Option 1:double=0.5, Index="Equal (Float)"(A=@n10)) -> (ReturnValue:double)
    Enable Unassign Button:bool=true
   ) -> (then, ReturnValue:Fort Dialogue Data Object Reference)
   > Set Dialogue Data(Dialogue Data:Fort Dialogue Data Object Reference=@n11) -> (Output_Get:Fort Dialogue Data Object Reference)
   Sequence(execute=@n9.then_1)
   > n12 = Branch(Condition=IsEmpty(InString=Get Sequence()))
   > n13 = Get!(execute:Exec) -> (then:Exec, else:Exec, Unassigned Lines Data Object:Fort Sound Sequence Data Object Reference)
   > Set Row Index!(
    execute
    Row Index:int=Length(TargetArray:Array of Fort Dialogue Data Object References=Get Dialogue Data!(self:Fort Sound Sequence Data Object Reference=@n13) -> (Dialogue Data:Array of Fort Dialogue Data Object References) {is=title K2Node_CallFunction})
    self:Fort Dialogue Data Object Reference=Get Dialogue Data() -> (Dialogue Data:Fort Dialogue Data Object Reference) {is=title K2Node_VariableGet}
   ) -> (then, Output_Get:int=0)
   > Set Enable Unassign Button!(execute, Enable Unassign Button:bool=false, self:Fort Dialogue Data Object Reference=Get Dialogue Data()) -> (then, Output_Get:bool)
   > Add Dialogue Data!(
    execute
    self:Fort Sound Sequence Data Object Reference=Get Unassigned Lines Data Object() -> (Unassigned Lines Data Object:Fort Sound Sequence Data Object Reference)
    Dialogue Data to Add:Fort Dialogue Data Object Reference=Get Dialogue Data()
   ) -> (then)
   n14 = Construct Object from Class!(
    execute=@n13.else
    Class=&FortSoundSequenceData_C
    self
    Row Index:int=0
    Asset Name:String=Unassigned Lines
    Dialogue Bumpers:Name=None
    Dialogue Data:Array of Fort Dialogue Data Object References=Make Array("[0]":Fort Dialogue Data Object Reference=Get Dialogue Data()) -> (Array:Array of Fort Dialogue Data Object References)
    Enable Delete:bool=False
   ) -> (then, ReturnValue:Fort Sound Sequence Data Object Reference)
   > Set Unassigned Lines Data Object(Unassigned Lines Data Object:Fort Sound Sequence Data Object Reference=@n14) -> (Output_Get:Fort Sound Sequence Data Object Reference)
   n15 = Sequence(execute=@n12.else)
   > Set Existing Sound Sequence(Existing Sound Sequence:Fort Sound Sequence Data Object Reference=Find(TargetMap:Map of Strings to Fort Sound Sequence Data Object References=Get Local Sound Sequence Objects(), Key=Get Sequence()) -> (Value:Fort Sound Sequence Data Object Reference)) -> (Output_Get:Fort Sound Sequence Data Object Reference)
   Sequence(execute=@n15.then_1)
   > n16 = Get!(execute) -> (then, else, Existing Sound Sequence:Fort Sound Sequence Data Object Reference)
   > Set Row Index(
    Row Index=Length(TargetArray=Get Dialogue Data(self=@n16) {is=title K2Node_CallFunction})
    self=Get Dialogue Data() {is=title K2Node_VariableGet}
   ) -> (Output_Get=0)
   > Add Dialogue Data(
    self=Get Existing Sound Sequence() -> (Existing Sound Sequence:Fort Sound Sequence Data Object Reference)
    Dialogue Data to Add=Get Dialogue Data()
   )
   n17 = Construct Object from Class(
    execute=@n16.else
    Class=&FortSoundSequenceData_C
    Row Index=Length!(TargetMap:Map of Strings to Fort Sound Sequence Data Object References=Get Local Sound Sequence Objects()) -> (ReturnValue:int=0) {is=title K2Node_CallFunction}
    Asset Name=Get Sequence()
    Dialogue Bumpers=No Bumpers
    Dialogue Data=Make Array("[0]"=Get Dialogue Data())
    Enable Delete=False
   )
   > Add!(
    execute
    TargetMap:Map of Strings to Fort Sound Sequence Data Object References=Get Local Sound Sequence Objects()
    Key:"String (by ref)"=Get Sequence()
    Value:"Fort Sound Sequence Data Object Reference (by ref)"=@n17
   ) -> (then) {is=title K2Node_CallFunction}
  ]}
 Create Dialogue Sound Cues {nodes [
   Create Dialogue Sound Cues() {is=title K2Node_FunctionEntry}
   n1 = Get Script Data Table()
   > n2 = Locate Assets in Path!(
    execute
    Path:Name=String To Name(InString=Get Subfolder Path(Asset=@n1, Subfolder Name=DialogueWaves))
    Class:Object Class Reference=/Script/Engine.DialogueWave
    Filename Suffix:String=_Dialogue
    __WorldContext:Object Reference
   ) -> (then, Success:bool, Found Assets:Array of Soft Object References)
   > n3 = For Each Loop(Array:Array of Soft Object References=@n2.Found Assets) -> (Array Element:Soft Object Reference)
   > Create Sound Cue from Dialogue Wave!(
    execute
    Dialogue Wave:Soft Object Reference=@n3
    Destination Path:String=Get Subfolder Path(Asset=Get Script Data Table(), Subfolder Name=SoundCues)
    __WorldContext:Object Reference
   ) -> (then)
  ]}
 Create Dialogue Wave Assets {nodes [
   Create Dialogue Wave Assets() {is=title K2Node_FunctionEntry}
   n1 = Get Script Data Table()
   > n2 = Locate Assets in Path(Path=String To Name(InString=Get Subfolder Path(Asset=@n1, Subfolder Name=Waves)), Class=/Script/Engine.SoundWave)
   > n3 = Set Found Sound Waves(Found Sound Waves:Array of Soft Object References=@n2.Found Assets) -> (Output_Get:Array of Soft Object References)
   > n4 = For Each Loop(Array=@n3)
   > n5 = Set Sound Wave(Sound Wave:Soft Object Reference=@n4) -> (Output_Get:Soft Object Reference)
   > n6 = Get Data Table Row(
    DataTable=Get Script Data Table()
    RowName=String To Name(InString=GetSoftPtrAssetName!(Asset:Soft Object Reference=@n5) -> (ReturnValue:String))
   )
   > n7 = Sequence()
   > Set Subtitle(Subtitle:String=Break VO Dialogue Wave Struct!(VO_DialogueWave_Struct=@n6) -> (Dialogue_15_#2)) -> (Output_Get:String)
   n8 = Sequence(execute=@n7.then_1)
   n9 = Get Sound Wave() -> (Sound Wave:Soft Object Reference)
   > Set Asset Name(Asset Name:String=Append(A=GetSoftPtrAssetName(Asset=@n9), B=_Dialogue)) -> (Output_Get:String)
   > Set Destination Path(Destination Path:String=Get Subfolder Path(Asset=Get Script Data Table(), Subfolder Name=DialogueWaves)) -> (Output_Get:String)
   Sequence(execute=@n8.then_1)
   > Create Dialogue Wave Asset!(
    execute
    Destination Path:String=Get Destination Path() -> (Destination Path:String)
    Asset Name:String=Get Asset Name!() -> (Asset Name)
    Dialogue Voice:Dialogue Voice Object Reference=Get Default Dialogue Voice() -> (Default Dialogue Voice:Dialogue Voice Object Reference)
    Sound Wave:Soft Object Reference=Get Sound Wave()
    Subtitle:String=Get Subtitle() -> (Subtitle:String)
    __WorldContext:Object Reference
   ) -> (then, Success:bool)
  ]}
 Create Sound Sequence Assets {nodes [
   Create Sound Sequence Assets() {is=title K2Node_FunctionEntry}
   > n1 = Construct Object from Class!(
    execute
    Class=/Game/Sounds/Fort_Audio_Blueprints/Utilities/EditorUtilityWidgets/VO_Utilities/Tasks/EUT_CreateSoundSequenceAssets.EUT_CreateSoundSequenceAssets_C
    self
    Sound Sequence Data:Array of Fort Sound Sequence Data Object References=Get Sound Sequence Data Objects()
    Destination Folder Path:String=Get Subfolder Path(Asset=Get Script Data Table(), Subfolder Name=SoundSequences)
   ) -> (then, ReturnValue:Task Create Sound Sequence Assets Object Reference)
   > RegisterAndExecuteTask!(
    execute
    self:Editor Utility Subsystem Object Reference=Get EditorUtilitySubsystem()
    NewTask:Editor Utility Task Object Reference=@n1
    OptionalParentTask:Editor Utility Task Object Reference
   ) -> (then)
  ]}
 Dialogue List Events {nodes [
   n1 = "On Delay Amount Changed (DialogueEditor)"!() -> (OutputDelegate, then, Dialogue Data Object:Fort Dialogue Data Object Reference, New Delay Amount:double) {is=title K2Node_ComponentBoundEvent}
   > Set Dialogue Post-sentence Delay Amount!(
    execute
    self:Fort Sound Sequence Data Object Reference=Get Selected Sound Sequence Data() -> (Selected Sound Sequence Data:Fort Sound Sequence Data Object Reference)
    Dialogue Data:Fort Dialogue Data Object Reference=@n1.Dialogue Data Object
    Delay Amount:double=@n1.New Delay Amount
   ) -> (then) {is=title K2Node_CallFunction}
   n2 = "On Sound Changed (DialogueEditor)"!() -> (OutputDelegate, then, Dialogue Data Object:Fort Dialogue Data Object Reference, New Sound:Sound Base Object Reference) {is=title K2Node_ComponentBoundEvent}
   > Set Dialogue Sound!(
    execute
    self:Fort Sound Sequence Data Object Reference=Get Selected Sound Sequence Data()
    Dialogue Data:Fort Dialogue Data Object Reference=@n2.Dialogue Data Object
    Sound:Sound Base Object Reference=@n2.New Sound
   ) -> (then) {is=title K2Node_CallFunction}
   n3 = "On Line Unassigned (DialogueEditor)"!() -> (OutputDelegate, then, Dialogue Data Object:Fort Dialogue Data Object Reference) {is=title K2Node_ComponentBoundEvent}
   > Reassign Dialogue Line!(
    execute
    self
    Dialogue to Reassign:Fort Dialogue Data Object Reference=@n3.Dialogue Data Object
    Target Sound Sequence:Fort Sound Sequence Data Object Reference=Get Unassigned Lines Data Object()
   ) -> (then)
   n4 = "On Sound Sequence Name Changed (DialogueEditor)"!() -> (OutputDelegate, then, Sound Sequence Name:String) {is=title K2Node_ComponentBoundEvent}
   > Set Asset Name!(execute, Asset Name=@n4.Sound Sequence Name, self:Fort Sound Sequence Data Object Reference=Get Selected Sound Sequence Data()) -> (then, Output_Get)
   n5 = "On Bumper Selection Changed (DialogueEditor)"!() -> (OutputDelegate, then, Selected Bumper:String) {is=title K2Node_ComponentBoundEvent}
   > Set Dialogue Bumpers!(
    execute
    self:Fort Sound Sequence Data Object Reference=Get Selected Sound Sequence Data()
    Key:String=@n5.Selected Bumper
    Dialogue Bumpers Data Asset:"Fort Dialogue Bumpers Data Object Reference (by ref)"=Get Dialogue Bumpers Data Asset() -> (Dialogue Bumpers Data Asset:Fort Dialogue Bumpers Data Object Reference)
   ) -> (then) {is=title K2Node_CallFunction}
   > Refresh List(self:Fort Sound Sequence Dialogue Viewer Object Reference=Get DialogueEditor() -> (DialogueEditor:Fort Sound Sequence Dialogue Viewer Object Reference))
   "On Preview Playback Started (DialogueEditor)"() {is=title K2Node_ComponentBoundEvent}
   > SetIsEnabled(self=Get MainControls() -> (MainControls:Vertical Box Object Reference), bInIsEnabled=false)
   "On Preview Playback Finished (DialogueEditor)"() {is=title K2Node_ComponentBoundEvent}
   > SetIsEnabled(self=Get MainControls(), bInIsEnabled=true)
   n6 = "On Lines Reordered (DialogueEditor)"!() -> (
    OutputDelegate
    then
    Dialogue Data to Insert:Fort Dialogue Data Object Reference
    Dialogue Data to Insert Above:Fort Dialogue Data Object Reference
   ) {is=title K2Node_ComponentBoundEvent}
   > Reorder Dialogue Data!(
    execute
    self:Fort Sound Sequence Data Object Reference=Get Selected Sound Sequence Data()
    Dialogue Data to Insert:Fort Dialogue Data Object Reference=@n6.Dialogue Data to Insert
    Dialogue Data to Insert Above:Fort Dialogue Data Object Reference=@n6.Dialogue Data to Insert Above
   ) -> (then)
   > Refresh List(self=Get DialogueEditor())
   "On Bumper Settings Clicked (DialogueEditor)"() {is=title K2Node_ComponentBoundEvent}
   > n7 = GetPathNameForLoadedAsset(LoadedAsset=Get Dialogue Bumpers Data Asset())
   > SyncBrowserToObjects(AssetPaths=Make Array("[0]":String=@n7) -> (Array:Array of Strings))
  ]}
 EventGraph {nodes [
   "On Property Changed (DataTablePicker)"!() -> (OutputDelegate, then, PropertyName:Name=None) {is=title K2Node_ComponentBoundEvent}
   > n1 = Branch(Condition=Should Confirm Data Table Change Action!() -> (ReturnValue:bool))
   > K2Node_Composite_0 = Open Confirmation Dialog!(execute) -> (Ok:Exec, Cancel:Exec) {is=title K2Node_Composite}
   Set Script Data Table(
    execute=@K2Node_Composite_0.Cancel
    Script Data Table:Data Table Object Reference=Get Script Data Table Cache() -> (Script Data Table Cache:Data Table Object Reference)
   ) -> (Output_Get:Data Table Object Reference)
   ClearListItems(execute=[@K2Node_Composite_0.Ok, @n1.else], self=Get SoundSequenceList())
   > Reset Variables()
   > Create Data Objects from Data Table() {is=title K2Node_CallFunction}
   > Set List Items!(execute, self:List View Object Reference=Get SoundSequenceList(), InListItems:Array of Object References=Get Sound Sequence Data Objects()) -> (then) {is=title K2Node_CallFunction}
   > SetSelectedIndex!(execute, self:List View Object Reference=Get SoundSequenceList(), Index:int=0) -> (then)
   > Set Script Data Table Cache(Script Data Table Cache:Data Table Object Reference=Get Script Data Table()) -> (Output_Get:Data Table Object Reference)
   n2 = "On Item Selection Changed (SoundSequenceList)"!() -> (OutputDelegate, then, Item:Object Reference, bIsSelected:bool) {is=title K2Node_ComponentBoundEvent}
   > Select Sound Sequence!(
    execute
    self
    New Selected Sound Sequence:Fort Sound Sequence Data Object Reference=Cast To FortSoundSequenceData!(Object=@n2.Item) -> (AsFort Sound Sequence Data:Fort Sound Sequence Data Object Reference, bSuccess:bool=~) {is=title K2Node_DynamicCast}
   ) -> (then)
   > SetCheckedState!(
    execute
    self:Check Box Object Reference=Get ToggleButton(self=Get ViewUnassignedLinesButton() -> (ViewUnassignedLinesButton:Fort Sound Sequence Editor Toggle Button Object Reference))
    InCheckedState:ECheckBoxState Enum=Unchecked
   ) -> (then)
   n3 = "On Check State Changed (EnableSoundSequenceRemovalButton)"!() -> (OutputDelegate, then, Is Checked:bool) {is=title K2Node_ComponentBoundEvent}
   > n4 = For Each Loop(Array:Array of Fort Sound Sequence Data Object References=Get Sound Sequence Data Objects()) -> (Array Element:Fort Sound Sequence Data Object Reference)
   > Set Enable Delete!(execute, Enable Delete:bool=@n3.Is Checked, self:Fort Sound Sequence Data Object Reference=@n4) -> (then, Output_Get:bool)
   "On Button Pressed (AddSoundSequenceButton)"() {is=title K2Node_ComponentBoundEvent}
   > Add New Sound Sequence() {is=title K2Node_CallFunction}
   "On Clicked (CreateAssetsButton)"() {is=title K2Node_ComponentBoundEvent}
   > Get Assets To Overwrite() {is=title K2Node_CallFunction}
   > n5 = Branch(Condition=Is Empty(TargetArray=Get Assets To Overwrite() -> (Assets To Overwrite:Array of Strings) {is=title K2Node_VariableGet}))
   Create Sound Sequence Assets(execute=[@K2Node_Composite_1.Completed, @n5.then]) {is=title K2Node_CallFunction}
   > Set Script Data Table()
   > Reset Variables()
   > Set Sound Sequence!(
    execute
    self:Fort Sound Sequence Dialogue Viewer Object Reference=Get DialogueEditor()
    Sound Sequence Data:Fort Sound Sequence Data Object Reference
   ) -> (then) {is=title K2Node_CallFunction}
   K2Node_Composite_1 = Confirm Overwrite Assets!(execute=@n5.else) -> (Completed:Exec) {is=title K2Node_Composite}
   n6 = "On Entry Generated (SoundSequenceList)"!() -> (OutputDelegate, then, Widget:User Widget Object Reference) {is=title K2Node_ComponentBoundEvent}
   > n7 = Cast To FortSoundSequenceEntry!(execute, Object=@n6.Widget) -> (then, CastFailed, AsFort Sound Sequence Entry:Fort Sound Sequence Entry Object Reference) {is=title K2Node_DynamicCast}
   n8 = Delete Sound Sequence!() -> (OutputDelegate, then, Sound Sequence Data:Fort Sound Sequence Data Object Reference) {is=title K2Node_CustomEvent}
   n9 = Bind Event to On Delete Sound Sequence Clicked(execute=@n7.then, self:Fort Sound Sequence Entry Object Reference=@n7, Delegate=@n8) {is=title K2Node_AddDelegate}
   n10 = Reassign Dialogue Data!() -> (
    OutputDelegate
    then
    Dialogue To Reassign:Fort Dialogue Data Object Reference
    Target Sound Sequence:Fort Sound Sequence Data Object Reference
   ) {is=title K2Node_CustomEvent}
   Bind Event to On Reassign Dialogue Data(execute=@n9.then, self:Fort Sound Sequence Entry Object Reference=@n7, Delegate=@n10) {is=title K2Node_AddDelegate}
   "On Clicked (ImportSoundWavesButton)"() {is=title K2Node_ComponentBoundEvent}
   > Import Sound Waves()
   > Create Dialogue Wave Assets() {is=title K2Node_CallFunction}
   > Create Dialogue Sound Cues() {is=title K2Node_CallFunction}
   > Set Sounds() {is=title K2Node_CallFunction}
   > SetCheckedState(
    self=Get ToggleButton(self=Get ImportSoundWavesToolButton() -> (ImportSoundWavesToolButton:Fort Sound Sequence Editor Toggle Button Object Reference))
    InCheckedState=Unchecked
   )
   > RegenerateAllEntries(self:List View Base Object Reference=Get DialogueListView!(self:Fort Sound Sequence Dialogue Viewer Object Reference=Get DialogueEditor()) -> (DialogueListView:Editor Utility List View Object Reference))
   n11 = "On Entry Released (SoundSequenceList)"!() -> (OutputDelegate, then, Widget:User Widget Object Reference) {is=title K2Node_ComponentBoundEvent}
   > n12 = Cast To FortSoundSequenceEntry(Object=@n11.Widget)
   > Reset(self:Fort Sound Sequence Entry Object Reference=@n12)
   n13 = "On Check State Changed (ViewUnassignedLinesButton)"!() -> (OutputDelegate, then, Is Checked:bool) {is=title K2Node_ComponentBoundEvent}
   > n14 = Branch(Condition=@n13.Is Checked)
   > Set Is Selected!(
    execute
    Is Selected:bool=false
    self:Fort Sound Sequence Data Object Reference=Get Selected Sound Sequence Data Cache() -> (Selected Sound Sequence Data Cache:Fort Sound Sequence Data Object Reference)
   ) -> (then, Output_Get:bool)
   > Clear Selection(self:List View Object Reference=Get SoundSequenceList())
   > Set Sound Sequence(self=Get DialogueEditor(), Sound Sequence Data=Get Unassigned Lines Data Object())
   > SetCheckedState(self=Get ToggleButton(self=Get ViewUnassignedLinesButton()), InCheckedState=Checked)
   Set Sound Sequence(execute=@n14.else, self=Get DialogueEditor(), Sound Sequence Data=Get Selected Sound Sequence Data Cache())
   > Set Selected Item!(execute, self:List View Object Reference=Get SoundSequenceList(), Item:Object Reference=Get Selected Sound Sequence Data Cache()) -> (then) {is=title K2Node_CallFunction}
   Event PreConstruct()
   > SetObject(
    self=[
     Get ImportSoundWavesProperty() -> (ImportSoundWavesProperty:Single Property View Object Reference)
     Get DataTablePicker() -> (DataTablePicker:Single Property View Object Reference)
    ]
    NewObject=Self-Reference()
   )
   Event Construct()
   > Set Dialogue Bumpers Data Asset!(
    execute
    self:Fort Sound Sequence Dialogue Viewer Object Reference=Get DialogueEditor()
    Dialogue Bumpers Data Asset:Fort Dialogue Bumpers Data Object Reference=Get Dialogue Bumpers Data Asset()
   ) -> (then) {is=title K2Node_CallFunction}
   Set Sound Sequence(execute=@n15.then, self=Get DialogueEditor(), Sound Sequence Data=Get Selected Sound Sequence Data())
   Reassign Dialogue Line(execute=@n10.then, Dialogue to Reassign=@n10.Dialogue To Reassign, Target Sound Sequence=@n10.Target Sound Sequence)
   n15 = Remove Sound Sequence!(execute=@n8.then, self, Sound Sequence Data:Fort Sound Sequence Data Object Reference=@n8.Sound Sequence Data) -> (then)
  ]}
 Open Confirmation Dialog {
  path=EventGraph.K2Node_Composite_0.Open Confirmation Dialog
  nodes [
   Inputs()
   > n1 = Show Message Dialog(
    Title=Format Text(Format="LOCTEXT(\"Sound Sequence Editor\")")
    Message=Format Text(Format="LOCTEXT(\"Current changes will not be saved.\")")
    MessageType=OkCancel
    DefaultValue=Ok
    MessageCategory=Warning
   )
   > n2 = Switch on EAppReturnType(Selection=@n1)
   Outputs!(Ok:Exec=@n2.Ok, Cancel:Exec=@n2.Cancel)
  ]
 }
 Confirm Overwrite Assets {
  path=EventGraph.K2Node_Composite_1.Confirm Overwrite Assets
  nodes [
   Inputs()
   > n1 = Show Message Dialog(
    Title=Format Text(Format="LOCTEXT(\"Sound Sequence Editor\")")
    Message=Format Text(Format="LOCTEXT(\"Existing assets will be overwritten. Proceed?\")")
    MessageType=YesNo
    DefaultValue=Yes
    MessageCategory=Warning
   )
   > n2 = Switch on EAppReturnType(Selection=@n1)
   n3 = For Each Loop(Exec=@n2.Yes, Array:Array of Strings=Get Assets To Overwrite()) -> (Array Element:String)
   Outputs!(Completed:Exec=@n3.Completed)
   > DeleteAsset!(execute, AssetPathToDelete:String=@n3) -> (then, ReturnValue:bool)
  ]
 }
 Get Assets To Overwrite {nodes [
   Get Assets To Overwrite() {is=title K2Node_FunctionEntry}
   > Clear(TargetArray=Get Assets To Overwrite() {is=title K2Node_VariableGet})
   > n1 = ListAssets!(
    execute
    DirectoryPath:String=Get Subfolder Path(Asset=Get Script Data Table(), Subfolder Name=SoundSequences)
    bRecursive:bool=true
    bIncludeFolder:bool=false
   ) -> (then, ReturnValue:Array of Strings)
   > n2 = For Each Loop(Array=@n1)
   > n3 = For Each Loop(Array:Array of Fort Sound Sequence Data Object References=Get Sound Sequence Data Objects()) -> (Array Element:Fort Sound Sequence Data Object Reference)
   > Branch(Condition="Equal Exactly (String)"(
     A=Get Asset Name!(self:Fort Sound Sequence Data Object Reference=@n3) -> (Asset Name)
     B=Split!(InPath:String=@n2) -> (PathPart:String, FilenamePart:String, ExtensionPart:String).FilenamePart
    ))
   > Add(TargetArray=Get Assets To Overwrite(), NewItem=@n2) {is=title K2Node_CallArrayFunction}
  ]}
 Get Create Assets Button Text {nodes [
   Get Create Assets Button Text() {is=title K2Node_FunctionEntry}
   n1 = Length(TargetArray:Array of Fort Sound Sequence Data Object References=Get Sound Sequence Data Objects()) {is=title K2Node_CallArrayFunction}
   n2 = Select(Option 0:Text="LOCTEXT(\"s\")", Option 1:Text, Index="Equal (Integer)"(A=@n1, B=1)) -> (ReturnValue:Text)
   > Return Node(ReturnValue:Text=Select(
     Option 0="LOCTEXT(\"No Sound Sequence Assets\")"
     Option 1=Format Text!(Format="LOCTEXT(\"Create {num} Sound Sequence Asset{s}\")", num:int=@n1, s:Text=@n2) -> (Result)
     Index="integer > integer"(A=@n1)
    ))
  ]}
 Get Create Assets Button Tooltip Text {nodes [
   Get Create Assets Button Tooltip Text() {is=title K2Node_FunctionEntry}
   n1 = Is Not Empty(TargetArray:Array of Strings=Get Duplicate Names())
   n2 = Format Text!(Format="LOCTEXT(\"Cannot create Sound Sequence assets with duplicate names.\")") -> (Result)
   > Return Node(ReturnValue=SelectText(
     B=SelectText(A=@n2, B="LOCTEXT(\"No Sound Sequence assets to create.\")", bPickA=@n1)
     bPickA=Can Create Assets!() -> (ReturnValue:bool) {is=title K2Node_CallFunction}
    ))
  ]}
 Get Data Table Picker Visibility {nodes [
   Get Data Table Picker Visibility() {is=title K2Node_FunctionEntry}
   n1 = NOT Boolean(A=Is Valid Script Data Table!(Data Table:Data Table Object Reference=Get Script Data Table()) -> (Is Valid:bool))
   n2 = IsChecked(self=Get ToggleButton(self=Get ShowScriptPickerToolButton() -> (ShowScriptPickerToolButton:Fort Sound Sequence Editor Toggle Button Object Reference)))
   > Return Node(ReturnValue:ESlateVisibility Enum=Select(Option 0:ESlateVisibility Enum=Collapsed, Option 1:ESlateVisibility Enum=Visible, Index=OR Boolean(A=@n1, B=@n2)) -> (ReturnValue:ESlateVisibility Enum))
  ]}
 Get Import Info Visibility {nodes [
   Get Import Info Visibility() {is=title K2Node_FunctionEntry}
   n1 = Get!() -> (Import Sound Waves Path_Path:String)
   > Return Node(ReturnValue=Select(Option 0=Collapsed, Option 1=Visible, Index="Not Equal, Case Insenstive (String)"!(A:String=@n1, B:String) -> (ReturnValue:bool)))
  ]}
 Get Import Sound Waves Visibility {nodes [
   Get Import Sound Waves Visibility() {is=title K2Node_FunctionEntry}
   n1 = IsChecked(self=Get ToggleButton(self=Get ImportSoundWavesToolButton()))
   n2 = Is Valid Script Data Table(Data Table=Get Script Data Table())
   > Return Node(ReturnValue=Select(Option 0=Collapsed, Option 1=Visible, Index=AND Boolean(A=@n2, B=@n1)))
  ]}
 Get Is Import Assets Enabled {nodes [
   Get Is Import Assets Enabled() {is=title K2Node_FunctionEntry}
   n1 = Break Directory Path!(DirectoryPath:"Directory Path Structure (by ref)"=Get Import Sound Waves Path() -> (Import Sound Waves Path:Directory Path Structure)) -> (Path:String) {is=title K2Node_BreakStruct}
   > Return Node(ReturnValue:bool=NOT Boolean(A=IsEmpty(InString=@n1)))
  ]}
 Get No Sound Sequences Message Visibility {nodes [
   Get No Sound Sequences Message Visibility() {is=title K2Node_FunctionEntry}
   n1 = GetNumItems!(self:List View Object Reference=Get SoundSequenceList()) -> (ReturnValue:int=0)
   > Return Node(ReturnValue:ESlateVisibility Enum=Select(Option 0=Collapsed, Option 1=Visible, Index="Equal (Integer)"(A=@n1)))
  ]}
 Get Select As-rec Message Visibility {nodes [
   Get Select As-rec Message Visibility() {is=title K2Node_FunctionEntry}
   n1 = Get Script Data Table()
   > Return Node(ReturnValue=Select(Option 0=Visible, Option 1=Collapsed, Index=Is Valid Script Data Table(Data Table=@n1)))
  ]}
 Get Sound Sequence List Visibility {nodes [
   Get Sound Sequence List Visibility() {is=title K2Node_FunctionEntry}
   n1 = GetNumItems(self=Get SoundSequenceList()) -> (ReturnValue=0)
   > Return Node(ReturnValue=Select(Option 0=Collapsed, Option 1=Visible, Index="integer > integer"(A=@n1)))
  ]}
 Get Toolbar Visibility {nodes [
   Get Toolbar Visibility() {is=title K2Node_FunctionEntry}
   n1 = Get Script Data Table()
   > Return Node(ReturnValue=Select(Option 0=Collapsed, Option 1=Visible, Index=Is Valid Script Data Table(Data Table=@n1)))
  ]}
 Import Sound Waves {nodes [
   Import Sound Waves() {is=title K2Node_FunctionEntry}
   > Import Sound Waves!(
    execute
    Source Folder:String=Break Directory Path(DirectoryPath=Get Import Sound Waves Path())
    Destination Folder:String=Get Subfolder Path(Asset=Get Script Data Table(), Subfolder Name=Waves)
    __WorldContext:Object Reference
   ) -> (then) {is=title K2Node_CallFunction}
  ]}
 Is Valid Script Data Table {nodes [
   n1 = Is Valid Script Data Table!() -> (then, Data Table:Data Table Object Reference) {is=title K2Node_FunctionEntry}
   > Return Node!(
    execute
    Is Valid:bool="Equal (Object)"(
     A=GetDataTableRowStruct!(Table:Data Table Object Reference=@n1) -> (ReturnValue:Script Struct Object Reference)
     B=/Game/Sounds/Fort_Audio_Blueprints/Utilities/EditorUtilityWidgets/Types/VO_DialogueWave_Struct
    )
   )
  ]}
 Locate Sound {nodes [
   Locate Sound!() -> (then, Sounds Folder:String, Row Name:Name) {is=title K2Node_FunctionEntry}
   > n1 = Sequence()
   n2 = Get Row Name() -> (Row Name=None)
   > Set Sound Path(Sound Path:String=Append!(A=Get Sounds Folder() -> (Sounds Folder:String), B=/, C:String="To String (Name)"(InName=@n2), D:String=_Dialogue_Cue) -> (ReturnValue)) -> (Output_Get:String)
   Sequence(execute=@n1.then_1)
   > n3 = DoesAssetExist(AssetPath=Get Sound Path() -> (Sound Path:String))
   > n4 = Branch(Condition=@n3)
   > n5 = LoadAsset(AssetPath=Get Sound Path())
   > n6 = Cast To SoundBase!(execute, Object=@n5) -> (then, CastFailed, AsSound Base:Sound Base Object Reference) {is=title K2Node_DynamicCast}
   > Return Node!(execute, Sound:Sound Base Object Reference=@n6)
   Return Node(execute=[@n6.CastFailed, @n4.else])
  ]}
 Reassign Dialogue Line {nodes [
   Reassign Dialogue Line!() -> (then, Dialogue to Reassign:Fort Dialogue Data Object Reference, Target Sound Sequence:Fort Sound Sequence Data Object Reference) {is=title K2Node_FunctionEntry}
   > Branch(Condition="Not Equal (Object)"(
     A:Object Reference=Get Selected Sound Sequence Data()
     B:Object Reference=Get Target Sound Sequence() -> (Target Sound Sequence:Fort Sound Sequence Data Object Reference)
    ) -> (ReturnValue:bool=~) {is=title K2Node_PromotableOperator})
   > n1 = Remove Dialogue Data!(
    execute
    self:Fort Sound Sequence Data Object Reference=Get Selected Sound Sequence Data()
    Dialogue Data to Remove:Fort Dialogue Data Object Reference=Get Dialogue to Reassign() -> (Dialogue to Reassign:Fort Dialogue Data Object Reference)
   ) -> (then, ReturnValue:bool)
   > n2 = Branch(Condition=@n1)
   Add Dialogue Data(execute=[@n4.then, @n2.then], self=Get Target Sound Sequence(), Dialogue Data to Add=Get Dialogue to Reassign())
   > Refresh List(self=Get DialogueEditor())
   n3 = Remove Dialogue Data(execute=@n2.else, self=Get Unassigned Lines Data Object(), Dialogue Data to Remove=Get Dialogue to Reassign())
   > n4 = Branch(Condition=@n3)
  ]}
 Remove Sound Sequence {nodes [
   n1 = Remove Sound Sequence!() -> (then, Sound Sequence Data:Fort Sound Sequence Data Object Reference) {is=title K2Node_FunctionEntry}
   > Append Array!(
    execute
    TargetArray:Array of Fort Dialogue Data Object References=Get Dialogue Data(self=Get Unassigned Lines Data Object()) {is=title K2Node_CallFunction}
    SourceArray:Array of Fort Dialogue Data Object References=Get Dialogue Data(self=@n1)
   ) -> (then)
   > RemoveItem!(
    execute
    self:List View Object Reference=Get SoundSequenceList()
    Item:Object Reference=Get Sound Sequence Data() -> (Sound Sequence Data:Fort Sound Sequence Data Object Reference)
   ) -> (then)
   > Remove Item!(
    execute
    TargetArray:Array of Fort Sound Sequence Data Object References=Get Sound Sequence Data Objects()
    Item:"Fort Sound Sequence Data Object Reference (by ref)"=Get Sound Sequence Data()
   ) -> (then, ReturnValue:bool)
  ]}
 Reset Variables {nodes [
   Reset Variables() {is=title K2Node_FunctionEntry}
   > n1 = Sequence()
   > Clear(TargetArray:Array of Fort Sound Sequence Data Object References=Get Sound Sequence Data Objects())
   > Set Unassigned Lines Data Object()
   > Set Script Data Table Cache()
   > Set Selected Sound Sequence Data(Selected Sound Sequence Data:Fort Sound Sequence Data Object Reference) -> (Output_Get:Fort Sound Sequence Data Object Reference)
   Sequence(execute=@n1.then_1)
   > SetCheckedState(self=Get ToggleButton(self=Get ShowScriptPickerToolButton()), InCheckedState=Unchecked)
   > SetCheckedState(self=Get ToggleButton(self=Get EnableSoundSequenceRemovalButton()), InCheckedState=Unchecked)
  ]}
 Select Sound Sequence {nodes [
   Select Sound Sequence!() -> (then, New Selected Sound Sequence:Fort Sound Sequence Data Object Reference) {is=title K2Node_FunctionEntry}
   > n1 = Sequence()
   > n2 = Get!(execute:Exec) -> (then:Exec, else:Exec, Selected Sound Sequence Data Cache:Fort Sound Sequence Data Object Reference)
   > Set Is Selected(Is Selected=false, self=@n2)
   n3 = Sequence(execute=@n1.then_1)
   > n4 = Set Selected Sound Sequence Data(Selected Sound Sequence Data=Get New Selected Sound Sequence() -> (New Selected Sound Sequence:Fort Sound Sequence Data Object Reference))
   > Is Valid(InputObject=@n4)
   > Set Is Selected(Is Selected=true, self=@n4)
   n5 = Sequence(execute=@n3.then_1)
   > Set Sound Sequence(self=Get DialogueEditor(), Sound Sequence Data=Get Selected Sound Sequence Data())
   Sequence(execute=@n5.then_1)
   > n6 = Get!(execute) -> (then, else, Selected Sound Sequence Data:Fort Sound Sequence Data Object Reference)
   > Set Selected Sound Sequence Data Cache(Selected Sound Sequence Data Cache:Fort Sound Sequence Data Object Reference=@n6) -> (Output_Get:Fort Sound Sequence Data Object Reference)
  ]}
 Set Sounds {nodes [
   Set Sounds() {is=title K2Node_FunctionEntry}
   > Set Sounds Folder(Sounds Folder:String=Get Subfolder Path(Asset=Get Script Data Table(), Subfolder Name=SoundCues)) -> (Output_Get:String)
   > n1 = For Each Loop(Array=Get Sound Sequence Data Objects())
   > n2 = For Each Loop(Array:Array of Fort Dialogue Data Object References=Get Dialogue Data(self=@n1)) -> (Array Element:Fort Dialogue Data Object Reference)
   > n3 = Set Dialogue Data(Dialogue Data=@n2)
   n4 = Get Character!(self:Fort Dialogue Data Object Reference=@n3) -> (Character)
   > Branch(Condition=AND Boolean(A="Not Equal Exactly (String)"(A=@n4, B=Start Sound), B="Not Equal Exactly (String)"(A=@n4, B=End Sound)))
   > n5 = Locate Sound(Sounds Folder=Get Sounds Folder(), Row Name=Get Row Name!(self:Fort Dialogue Data Object Reference=@n2) -> (Row Name=None)) {is=title K2Node_CallFunction}
   > n6 = Cast To SoundCue!(execute, Object=@n5) -> (then, CastFailed, AsSound Cue:Sound Cue Object Reference) {is=title K2Node_DynamicCast}
   > SetEditorProperty(
    Object=@n6
    PropertyName=SoundClassObject
    PropertyValue:Sound Class Object Reference=Get Sound Class() -> (Sound Class:Sound Class Object Reference)
   )
   > SetEditorProperty(
    Object=@n6
    PropertyName=AttenuationSettings
    PropertyValue:Sound Attenuation Object Reference=Get Attenuation Settings() -> (Attenuation Settings:Sound Attenuation Object Reference)
   )
   > SetEditorProperty(Object=@n6, PropertyName=VolumeMultiplier, PropertyValue:double=Get Default Volume Multiplier() -> (Default Volume Multiplier:double))
   > Set Sound!(execute, Sound=@n6, self:Fort Dialogue Data Object Reference=Get Dialogue Data() {is=title K2Node_VariableGet}) -> (then, Output_Get)
  ]}
 Should Confirm Data Table Change Action {nodes [
   Should Confirm Data Table Change Action() {is=title K2Node_FunctionEntry}
   n1 = Get Script Data Table Cache()
   n2 = Is Valid Script Data Table(Data Table=@n1) {is=title K2Node_CallFunction}
   n3 = "Not Equal (Object)"(A=@n1) -> (ReturnValue=~)
   n4 = Get Script Data Table()
   n5 = "Not Equal (Object)"(A=@n4, B=@n1) -> (ReturnValue=~)
   > Return Node!(execute, ReturnValue:bool=OR Boolean(A="Equal (Object)"(A=@n4), B=AND Boolean!(A=@n3, B=@n5, C:bool=@n2) -> (ReturnValue)))
  ]}
]
variables [
 Assets To Overwrite {type="array<string>", default []}
 Attenuation Settings {type=SoundAttenuation, default=/Game/Sounds/Fort_Audio_Blueprints/Utilities/EditorUtilityWidgets/VO_Utilities/Assets/AS_Fort_AudioLogs}
 Default Dialogue Voice {type=DialogueVoice, default=/Game/Sounds/Fort_Audio_Blueprints/Utilities/EditorUtilityWidgets/VO_Utilities/Assets/AudioLogVO_DV}
 Default End Sound {type=SoundBase, default=/Game/Sounds/Fort_Test_Sounds/Tones/1000hz_Triangle_Tone_Short}
 Default Start Sound {type=SoundBase, default=/Game/Sounds/Fort_Test_Sounds/Tones/1000hz_Saw_Tone_Short}
 Default Volume Multiplier {type=number, default=1}
 Dialogue Bumpers Data Asset {type=FortDialogueBumpersData_C, default=&/DialogueBumpers}
 Duplicate Names {type="array<string>", default []}
 Import Sound Waves Path {type=DirectoryPath, default {path=""}}
 Script Data Table {type=DataTable, default=null}
 Script Data Table Cache {type=DataTable, default=null}
 Selected Sound Sequence Data {type=FortSoundSequenceData_C, default=null}
 Selected Sound Sequence Data Cache {type=FortSoundSequenceData_C, default=null}
 Sound Class {type=SoundClass, default=/Game/Sounds/FortSoundClassesAndModes/Fort_VO_Narrator, class=~, graph_type=~, children=~, properties=~}
 Sound Sequence Data Objects {type="array<FortSoundSequenceData_C>", default []}
 Unassigned Lines Data Object {type=FortSoundSequenceData_C, default=null}
]
widgets {
 info {namedSlotCount=2, rootWidgetClass=&Border, widgetCount=47}
 widgets [
  Border_3 {
   widgetClassPath=&Border
   widget=Border_3
   properties {
    background {
     bIsDynamicallyLoaded=false
     drawAs=Image
     imageType=NoImage
     mirroring=NoMirror
     resourceName=None
     resourceObject=null
     tiling=NoTile
     imageSize=<32 32>
     margin=<ltrb 0 0 0 0>
     outlineSettings {
      bUseBrushTransparency=false
      roundingType=HalfHeightRadius
      width=0
      color {colorUseRule=UseColor_Specified, specifiedColor=<rgba 0 0 0 0>}
      cornerRadii=<0 0 0 0>
     }
     tintColor {colorUseRule=UseColor_Specified, specifiedColor=<rgba 0 0 0 0.1>}
     uVRegion {bIsValid=false, max=<0 0>, min=<0 0>}
    }
    padding=<ltrb 2 2 2 2>
   }
  }
  HorizontalBox_0 {
   widgetClassPath=&HorizontalBox
   parent=Border_3
   widget=HorizontalBox_0
   slot {class=BorderSlot, properties {padding=<ltrb 2 2 2 2>}}
  }
  SizeBox_0 {
   widgetClassPath=&SizeBox
   parent=HorizontalBox_0
   widget=SizeBox_0
   properties {bOverride_WidthOverride=true, widthOverride=320}
   slot {class=HorizontalBoxSlot}
  }
  Border_0 {
   widgetClassPath=&Border
   parent=SizeBox_0
   widget=Border_0
   properties {
    background {
     bIsDynamicallyLoaded=false
     drawAs=RoundedBox
     imageType=NoImage
     mirroring=NoMirror
     resourceName=None
     resourceObject=null
     tiling=NoTile
     imageSize=<32 32>
     margin=<ltrb 0 0 0 0>
     outlineSettings {
      bUseBrushTransparency=false
      roundingType=FixedRadius
      width=0
      color {colorUseRule=UseColor_Specified, specifiedColor=<rgba 0 0 0 0>}
      cornerRadii=<4 4 0 0>
     }
     tintColor {colorUseRule=UseColor_Specified, specifiedColor=<rgba 0.0103 0.0103 0.0103 1>}
     uVRegion {bIsValid=false, max=<0 0>, min=<0 0>}
    }
    padding=<ltrb 0 0 0 0>
   }
   slot {class=SizeBoxSlot}
  }
  MainControls {
   widgetClassPath=&VerticalBox
   bIsVariable=true
   parent=Border_0
   widget=MainControls
   slot {class=BorderSlot, properties {padding=<ltrb 0 0 0 0>}}
  }
  Title {widgetClassPath=&HorizontalBox, parent=MainControls, widget=Title, slot {class=VerticalBoxSlot}}
  Border_2 {
   widgetClassPath=&Border
   parent=Title
   widget=Border_2
   properties {
    horizontalAlignment=HAlign_Center
    background {
     bIsDynamicallyLoaded=false
     drawAs=Box
     imageType=NoImage
     mirroring=NoMirror
     resourceName=None
     resourceObject=null
     tiling=NoTile
     imageSize=<32 32>
     margin=<ltrb 0 0 0 0>
     outlineSettings {
      bUseBrushTransparency=false
      roundingType=FixedRadius
      width=0
      color {colorUseRule=UseColor_Specified, specifiedColor=<rgba 0 0 0 0>}
      cornerRadii=<4 4 0 0>
     }
     tintColor {colorUseRule=UseColor_Specified, specifiedColor=<rgba 0.0176 0.0176 0.0176 1>}
     uVRegion {bIsValid=false, max=<0 0>, min=<0 0>}
    }
    padding=<ltrb 4 4 4 4>
   }
   slot {class=HorizontalBoxSlot, properties {size {sizeRule=Fill, value=1}}}
  }
  TextBlock_96 {
   widgetClassPath=&TextBlock
   parent=Border_2
   widget=TextBlock_96
   properties {
    text=Dialogue Pipeline Tool
    colorAndOpacity {colorUseRule=UseColor_Specified, specifiedColor=<rgba 0.5029 0.5029 0.5029 1>}
    font {
     bForceMonospaced=false
     bMaterialIsStencil=false
     fontMaterial=null
     fontObject=&Roboto
     letterSpacing=0
     monospacedWidth=1
     size=12
     skewAmount=0
     typefaceFontName=Bold
     outlineSettings {
      bApplyOutlineToDropShadows=false
      bMiteredCorners=false
      bSeparateFillAlpha=false
      outlineMaterial=null
      outlineSize=0
      outlineColor=<rgba 0 0 0 1>
     }
    }
   }
   slot {class=BorderSlot, properties {horizontalAlignment=HAlign_Center, padding=<ltrb 4 4 4 4>}}
  }
  Toolbar {
   widgetClassPath=&HorizontalBox
   parent=MainControls
   widget=Toolbar
   slot {class=VerticalBoxSlot, properties {verticalAlignment=VAlign_Top}}
  }
  Border_1 {
   widgetClassPath=&Border
   parent=Toolbar
   widget=Border_1
   properties {
    background {
     bIsDynamicallyLoaded=false
     drawAs=Box
     imageType=NoImage
     mirroring=NoMirror
     resourceName=None
     resourceObject=null
     tiling=NoTile
     imageSize=<32 32>
     margin=<ltrb 0 0 0 0>
     outlineSettings {
      bUseBrushTransparency=false
      roundingType=FixedRadius
      width=0
      color {colorUseRule=UseColor_Specified, specifiedColor=<rgba 0 0 0 0>}
      cornerRadii=<0 0 4 4>
     }
     tintColor {colorUseRule=UseColor_Specified, specifiedColor=<rgba 0.0176 0.0176 0.0176 1>}
     uVRegion {bIsValid=false, max=<0 0>, min=<0 0>}
    }
    padding=<ltrb 4 0 4 4>
   }
   slot {class=HorizontalBoxSlot, properties {verticalAlignment=VAlign_Top, size {sizeRule=Fill, value=1}}}
  }
  HorizontalBox_114 {
   widgetClassPath=&HorizontalBox
   parent=Border_1
   widget=HorizontalBox_114
   slot {class=BorderSlot, properties {padding=<ltrb 4 0 4 4>}}
  }
  WrapBox_0 {
   widgetClassPath=&WrapBox
   parent=HorizontalBox_114
   widget=WrapBox_0
   properties {horizontalAlignment=HAlign_Center, innerSlotPadding=<4 4>}
   slot {class=HorizontalBoxSlot, properties {verticalAlignment=VAlign_Center, size {sizeRule=Fill, value=1}}}
  }
  ShowScriptPickerToolButton {
   widgetClassPath=/Game/Sounds/Fort_Audio_Blueprints/Utilities/EditorUtilityWidgets/VO_Utilities/Widgets/FortSoundSequenceEditorToggleButton
   bIsVariable=true
   parent=WrapBox_0
   widget=ShowScriptPickerToolButton
   properties {
    icon=/Game/Athena/UI/InGame/Creative/PropertyEditor/Icon_Picker/Textures_V2/64x64/T_UI_IconLibrary_SpeechBubble_64
    toolTip=Select As-Rec Data Table
   }
   slot {class=WrapBoxSlot}
  }
  ImportSoundWavesToolButton {
   widgetClassPath=&FortSoundSequenceEditorToggleButton
   bIsVariable=true
   parent=WrapBox_0
   widget=ImportSoundWavesToolButton
   properties {icon=/Game/UI/Foundation/Textures/Icons/Audio/T-Icon-Sound-On-M, is Enabled by Default=true, toolTip=Import Sound Waves}
   slot {class=WrapBoxSlot}
  }
  ViewUnassignedLinesButton {
   widgetClassPath=&FortSoundSequenceEditorToggleButton
   bIsVariable=true
   parent=WrapBox_0
   widget=ViewUnassignedLinesButton
   properties {icon=/FortniteConversationUI/Conversation/Materials/T_UI_NPC_SpeechMax, toolTip=View Unassigned Lines}
   slot {class=WrapBoxSlot}
  }
  AddSoundSequenceButton {
   widgetClassPath=/Game/Sounds/Fort_Audio_Blueprints/Utilities/EditorUtilityWidgets/VO_Utilities/Widgets/FortSoundSequenceEditorButton
   bIsVariable=true
   parent=WrapBox_0
   widget=AddSoundSequenceButton
   properties {icon=/Game/UI/Foundation/Textures/Icons/Replay/T-Icon-Replay-Plus-L, toolTip=Add New Sound Sequence}
   slot {class=WrapBoxSlot, properties {verticalAlignment=VAlign_Center}}
  }
  EnableSoundSequenceRemovalButton {
   widgetClassPath=&FortSoundSequenceEditorToggleButton
   bIsVariable=true
   parent=WrapBox_0
   widget=EnableSoundSequenceRemovalButton
   properties {icon=/Game/UI/Foundation/Textures/Buttons/T-SlotIcon-NotRecyclable-64, toolTip=Enable Sound Sequence Removal}
   slot {class=WrapBoxSlot}
  }
  DataTablePickerBox {
   widgetClassPath=&VerticalBox
   parent=MainControls
   widget=DataTablePickerBox
   slot {class=VerticalBoxSlot, properties {padding=<ltrb 4 4 4 4>}}
  }
  HorizontalBox_2 {
   widgetClassPath=&HorizontalBox
   parent=DataTablePickerBox
   widget=HorizontalBox_2
   slot {class=VerticalBoxSlot, properties {horizontalAlignment=HAlign_Center, verticalAlignment=VAlign_Center}}
  }
  ImportSoundWavesIcon_1 {
   widgetClassPath=&Image
   parent=HorizontalBox_2
   widget=ImportSoundWavesIcon_1
   properties {
    brush {
     bIsDynamicallyLoaded=false
     drawAs=Image
     imageType=FullColor
     mirroring=NoMirror
     resourceName=None
     resourceObject=&T_UI_IconLibrary_SpeechBubble_64
     tiling=NoTile
     imageSize=<12 12>
     margin=<ltrb 0 0 0 0>
     outlineSettings {
      bUseBrushTransparency=false
      roundingType=HalfHeightRadius
      width=0
      color {colorUseRule=UseColor_Specified, specifiedColor=<rgba 0 0 0 0>}
      cornerRadii=<0 0 0 0>
     }
     tintColor {colorUseRule=UseColor_Specified, specifiedColor=<rgba 1 1 1 1>}
     uVRegion {bIsValid=false, max=<0 0>, min=<0 0>}
    }
    colorAndOpacity=<rgba 0.5029 0.5029 0.5029 1>
   }
   slot {class=HorizontalBoxSlot, properties {horizontalAlignment=HAlign_Left, verticalAlignment=VAlign_Center, padding=<ltrb 0 0 4 0>}}
  }
  ImportSoundWavesText_1 {
   widgetClassPath=&TextBlock
   parent=HorizontalBox_2
   widget=ImportSoundWavesText_1
   properties {
    text=Select As-Rec Data Table
    colorAndOpacity {colorUseRule=UseColor_Specified, specifiedColor=<rgba 0.5029 0.5029 0.5029 1>}
    font {
     bForceMonospaced=false
     bMaterialIsStencil=false
     fontMaterial=null
     fontObject=&Roboto
     letterSpacing=0
     monospacedWidth=1
     size=9
     skewAmount=0
     typefaceFontName=Bold
     outlineSettings {
      bApplyOutlineToDropShadows=false
      bMiteredCorners=false
      bSeparateFillAlpha=false
      outlineMaterial=null
      outlineSize=0
      outlineColor=<rgba 0 0 0 1>
     }
    }
   }
   slot {class=HorizontalBoxSlot, properties {padding=<ltrb 4 4 4 4>}}
  }
  DataTablePicker {
   widgetClassPath=&SinglePropertyView
   parent=DataTablePickerBox
   widget=DataTablePicker
   properties {nameOverride=As-Rec, propertyName=Script Data Table}
   slot {class=VerticalBoxSlot, properties {size {sizeRule=Fill, value=1}}}
  }
  SelectAsRecDataTableMessage {
   widgetClassPath=&VerticalBox
   parent=MainControls
   widget=SelectAsRecDataTableMessage
   properties {visibility=Collapsed}
   slot {class=VerticalBoxSlot, properties {size {sizeRule=Fill, value=1}}}
  }
  TextBlock_229 {
   widgetClassPath=&TextBlock
   parent=SelectAsRecDataTableMessage
   widget=TextBlock_229
   properties {
    text=Select a valid as-rec data table.
    colorAndOpacity {colorUseRule=UseColor_Specified, specifiedColor=<rgba 0.5029 0.5029 0.5029 1>}
    font {
     bForceMonospaced=false
     bMaterialIsStencil=false
     fontMaterial=null
     fontObject=&Roboto
     letterSpacing=0
     monospacedWidth=1
     size=9
     skewAmount=0
     typefaceFontName=Regular
     outlineSettings {
      bApplyOutlineToDropShadows=false
      bMiteredCorners=false
      bSeparateFillAlpha=false
      outlineMaterial=null
      outlineSize=0
      outlineColor=<rgba 0 0 0 1>
     }
    }
   }
   slot {class=VerticalBoxSlot, properties {horizontalAlignment=HAlign_Center, verticalAlignment=VAlign_Center, size {sizeRule=Fill, value=1}}}
  }
  ImportSoundWavesBox {widgetClassPath=&HorizontalBox, parent=MainControls, widget=ImportSoundWavesBox, slot {class=VerticalBoxSlot}}
  VerticalBox {
   widgetClassPath=&VerticalBox
   parent=ImportSoundWavesBox
   widget=VerticalBox
   slot {class=HorizontalBoxSlot, properties {size {sizeRule=Fill, value=1}}}
  }
  ImportSoundWavesHeader {
   widgetClassPath=&VerticalBox
   parent=VerticalBox
   widget=ImportSoundWavesHeader
   slot {class=VerticalBoxSlot, properties {size {sizeRule=Fill, value=1}}}
  }
  HorizontalBox_1 {
   widgetClassPath=&HorizontalBox
   parent=ImportSoundWavesHeader
   widget=HorizontalBox_1
   slot {class=VerticalBoxSlot, properties {horizontalAlignment=HAlign_Center}}
  }
  ImportSoundWavesIcon {
   widgetClassPath=&Image
   parent=HorizontalBox_1
   widget=ImportSoundWavesIcon
   properties {
    brush {
     bIsDynamicallyLoaded=false
     drawAs=Image
     imageType=FullColor
     mirroring=NoMirror
     resourceName=None
     resourceObject=&T-Icon-Sound-On-M
     tiling=NoTile
     imageSize=<12 12>
     margin=<ltrb 0 0 0 0>
     outlineSettings {
      bUseBrushTransparency=false
      roundingType=HalfHeightRadius
      width=0
      color {colorUseRule=UseColor_Specified, specifiedColor=<rgba 0 0 0 0>}
      cornerRadii=<0 0 0 0>
     }
     tintColor {colorUseRule=UseColor_Specified, specifiedColor=<rgba 1 1 1 1>}
     uVRegion {bIsValid=false, max=<0 0>, min=<0 0>}
    }
    colorAndOpacity=<rgba 0.5029 0.5029 0.5029 1>
   }
   slot {class=HorizontalBoxSlot, properties {horizontalAlignment=HAlign_Left, verticalAlignment=VAlign_Center, padding=<ltrb 0 2 4 0>}}
  }
  ImportSoundWavesText {
   widgetClassPath=&TextBlock
   parent=HorizontalBox_1
   widget=ImportSoundWavesText
   properties {
    text=Import Sound Waves
    colorAndOpacity {colorUseRule=UseColor_Specified, specifiedColor=<rgba 0.5029 0.5029 0.5029 1>}
    font {
     bForceMonospaced=false
     bMaterialIsStencil=false
     fontMaterial=null
     fontObject=&Roboto
     letterSpacing=0
     monospacedWidth=1
     size=9
     skewAmount=0
     typefaceFontName=Bold
     outlineSettings {
      bApplyOutlineToDropShadows=false
      bMiteredCorners=false
      bSeparateFillAlpha=false
      outlineMaterial=null
      outlineSize=0
      outlineColor=<rgba 0 0 0 1>
     }
    }
   }
   slot {class=HorizontalBoxSlot, properties {padding=<ltrb 4 8 4 4>}}
  }
  ImportSoundWavesProperty {
   widgetClassPath=&SinglePropertyView
   parent=ImportSoundWavesHeader
   widget=ImportSoundWavesProperty
   properties {nameOverride=Import From, propertyName=Import Sound Waves Path}
   slot {class=VerticalBoxSlot, properties {padding=<ltrb 4 4 4 4>, size {sizeRule=Fill, value=1}}}
  }
  ImportSoundWavesButton {
   widgetClassPath=&EditorUtilityButton
   parent=ImportSoundWavesHeader
   widget=ImportSoundWavesButton
   properties {widgetStyle {
     clickedSlateSound {resourceObject=null}
     disabled {
      bIsDynamicallyLoaded=false
      drawAs=RoundedBox
      imageType=NoImage
      mirroring=NoMirror
      resourceName=None
      resourceObject=null
      tiling=NoTile
      imageSize=<0 0>
      margin=<ltrb 0 0 0 0>
      outlineSettings {
       bUseBrushTransparency=false
       roundingType=FixedRadius
       width=1
       color {colorUseRule=UseColor_ColorTable, specifiedColor=<rgba 0 0 0 0>}
       cornerRadii=<4 4 4 4>
      }
      tintColor {colorUseRule=UseColor_Specified, specifiedColor=<rgba 0.0395 0.0395 0.0395 1>}
      uVRegion {bIsValid=false, max=<0 0>, min=<0 0>}
     }
     disabledForeground {colorUseRule=UseColor_ColorTable, specifiedColor=<rgba 0 0 0 0>}
     hovered {
      bIsDynamicallyLoaded=false
      drawAs=RoundedBox
      imageType=NoImage
      mirroring=NoMirror
      resourceName=None
      resourceObject=null
      tiling=NoTile
      imageSize=<32 32>
      margin=<ltrb 0 0 0 0>
      outlineSettings {
       bUseBrushTransparency=false
       roundingType=FixedRadius
       width=1
       color {colorUseRule=UseColor_ColorTable, specifiedColor=<rgba 0 0 0 0>}
       cornerRadii=<4 4 4 4>
      }
      tintColor {colorUseRule=UseColor_Specified, specifiedColor=<rgba 0.0044 0.2384 1 1>}
      uVRegion {bIsValid=false, max=<0 0>, min=<0 0>}
     }
     hoveredForeground {colorUseRule=UseColor_ColorTable, specifiedColor=<rgba 0 0 0 0>}
     hoveredSlateSound {resourceObject=null}
     normal {
      bIsDynamicallyLoaded=false
      drawAs=RoundedBox
      imageType=NoImage
      mirroring=NoMirror
      resourceName=None
      resourceObject=null
      tiling=NoTile
      imageSize=<32 32>
      margin=<ltrb 0 0 0 0>
      outlineSettings {
       bUseBrushTransparency=false
       roundingType=FixedRadius
       width=1
       color {colorUseRule=UseColor_ColorTable, specifiedColor=<rgba 0 0 0 0>}
       cornerRadii=<4 4 4 4>
      }
      tintColor {colorUseRule=UseColor_Specified, specifiedColor=<rgba 0 0.162 0.7454 1>}
      uVRegion {bIsValid=false, max=<0 0>, min=<0 0>}
     }
     normalForeground {colorUseRule=UseColor_ColorTable, specifiedColor=<rgba 0 0 0 0>}
     normalPadding=<ltrb 12 1.5 12 1.5>
     pressed {
      bIsDynamicallyLoaded=false
      drawAs=RoundedBox
      imageType=NoImage
      mirroring=NoMirror
      resourceName=None
      resourceObject=null
      tiling=NoTile
      imageSize=<32 32>
      margin=<ltrb 0 0 0 0>
      outlineSettings {
       bUseBrushTransparency=false
       roundingType=FixedRadius
       width=1
       color {colorUseRule=UseColor_ColorTable, specifiedColor=<rgba 0 0 0 0>}
       cornerRadii=<4 4 4 4>
      }
      tintColor {colorUseRule=UseColor_Specified, specifiedColor=<rgba 0 0.0802 0.3515 1>}
      uVRegion {bIsValid=false, max=<0 0>, min=<0 0>}
     }
     pressedForeground {colorUseRule=UseColor_ColorTable, specifiedColor=<rgba 0 0 0 0>}
     pressedPadding=<ltrb 12 2.5 12 0.5>
     pressedSlateSound {resourceObject=null}
    }}
   slot {class=VerticalBoxSlot, properties {horizontalAlignment=HAlign_Center, padding=<ltrb 4 4 4 4>}}
   bIsVariable=~
  }
  TextBlock_119 {
   widgetClassPath=&TextBlock
   parent=ImportSoundWavesButton
   widget=TextBlock_119
   properties {
    text=Import Sound Waves
    font {
     bForceMonospaced=false
     bMaterialIsStencil=false
     fontMaterial=null
     fontObject=&Roboto
     letterSpacing=0
     monospacedWidth=1
     size=9
     skewAmount=0
     typefaceFontName=Bold
     outlineSettings {
      bApplyOutlineToDropShadows=false
      bMiteredCorners=false
      bSeparateFillAlpha=false
      outlineMaterial=null
      outlineSize=0
      outlineColor=<rgba 0 0 0 1>
     }
    }
    margin=<ltrb 0 1 0 0>
   }
   slot {class=ButtonSlot}
  }
  TextBlock_104 {
   widgetClassPath=&TextBlock
   parent=VerticalBox
   widget=TextBlock_104
   properties {
    autoWrapText=true
    justification=Center
    text=Assets will be imported to the same directory as the As-Rec data table asset.
    colorAndOpacity {colorUseRule=UseColor_Specified, specifiedColor=<rgba 0.5029 0.5029 0.5029 0.5>}
    font {
     bForceMonospaced=false
     bMaterialIsStencil=false
     fontMaterial=null
     fontObject=&Roboto
     letterSpacing=0
     monospacedWidth=1
     size=7.5
     skewAmount=0
     typefaceFontName=Regular
     outlineSettings {
      bApplyOutlineToDropShadows=false
      bMiteredCorners=false
      bSeparateFillAlpha=false
      outlineMaterial=null
      outlineSize=0
      outlineColor=<rgba 0 0 0 1>
     }
    }
   }
   slot {class=VerticalBoxSlot, properties {horizontalAlignment=HAlign_Center, verticalAlignment=VAlign_Center, padding=<ltrb 64 8 64 8>}}
  }
  SoundSequenceListBox {
   widgetClassPath=&HorizontalBox
   parent=MainControls
   widget=SoundSequenceListBox
   slot {class=VerticalBoxSlot, properties {padding=<ltrb 2 2 2 0>, size {sizeRule=Fill, value=1}}}
  }
  SoundSequencesExpandable {
   widgetClassPath=/Script/Blutility.EditorUtilityExpandableArea
   parent=SoundSequenceListBox
   widget=SoundSequencesExpandable
   properties {
    bIsExpanded=true
    borderBrush {
     bIsDynamicallyLoaded=false
     drawAs=RoundedBox
     imageType=NoImage
     mirroring=NoMirror
     resourceName=None
     resourceObject=null
     tiling=NoTile
     imageSize=<0 0>
     margin=<ltrb 0 0 0 0>
     outlineSettings {
      bUseBrushTransparency=false
      roundingType=FixedRadius
      width=1
      color {colorUseRule=UseColor_Specified, specifiedColor=<rgba 0 0 0 0>}
      cornerRadii=<4 4 4 4>
     }
     tintColor {colorUseRule=UseColor_Specified, specifiedColor=<rgba 0.0284 0.0284 0.0284 1>}
     uVRegion {bIsValid=false, max=<0 0>, min=<0 0>}
    }
    headerPadding=<ltrb 4 4 4 4>
   }
   slot {class=HorizontalBoxSlot, properties {size {sizeRule=Fill, value=1}}}
  }
  TextBlock_335 {
   widgetClassPath=&TextBlock
   namedSlotHost=SoundSequencesExpandable
   widget=TextBlock_335
   properties {
    text=Sound Sequences
    colorAndOpacity {colorUseRule=UseColor_Specified, specifiedColor=<rgba 0.5029 0.5029 0.5029 1>}
    font {
     bForceMonospaced=false
     bMaterialIsStencil=false
     fontMaterial=null
     fontObject=&Roboto
     letterSpacing=0
     monospacedWidth=1
     size=9.75
     skewAmount=0
     typefaceFontName=Bold
     outlineSettings {
      bApplyOutlineToDropShadows=false
      bMiteredCorners=false
      bSeparateFillAlpha=false
      outlineMaterial=null
      outlineSize=0
      outlineColor=<rgba 0 0 0 1>
     }
    }
    margin=<ltrb 0 2 0 0>
   }
  }
  Border_4 {
   widgetClassPath=&Border
   namedSlotHost=SoundSequencesExpandable
   widget=Border_4
   properties {brushColor=<rgba 0.006 0.006 0.006 1>, padding=<ltrb 0 0 0 0>}
  }
  VerticalBox_7 {widgetClassPath=&VerticalBox, parent=Border_4, widget=VerticalBox_7, slot {class=BorderSlot, properties {padding=<ltrb 0 0 0 0>}}}
  SoundSequenceList {
   widgetClassPath=/Script/Blutility.EditorUtilityListView
   bIsVariable=true
   parent=VerticalBox_7
   widget=SoundSequenceList
   properties {
    bClearSelectionOnClick=true
    bEnableScrollAnimation=true
    entryWidgetClass=/Game/Sounds/Fort_Audio_Blueprints/Utilities/EditorUtilityWidgets/VO_Utilities/Widgets/FortSoundSequenceEntry.FortSoundSequenceEntry_C
   }
   slot {class=VerticalBoxSlot, properties {size {sizeRule=Fill, value=1}}}
  }
  TextBlock {
   widgetClassPath=&TextBlock
   parent=VerticalBox_7
   widget=TextBlock
   properties {
    text=There are no sound sequences.
    colorAndOpacity {colorUseRule=UseColor_Specified, specifiedColor=<rgba 0.5029 0.5029 0.5029 1>}
    font {
     bForceMonospaced=false
     bMaterialIsStencil=false
     fontMaterial=null
     fontObject=&Roboto
     letterSpacing=0
     monospacedWidth=1
     size=9
     skewAmount=0
     typefaceFontName=Regular
     outlineSettings {
      bApplyOutlineToDropShadows=false
      bMiteredCorners=false
      bSeparateFillAlpha=false
      outlineMaterial=null
      outlineSize=0
      outlineColor=<rgba 0 0 0 1>
     }
    }
   }
   slot {class=VerticalBoxSlot, properties {horizontalAlignment=HAlign_Center, verticalAlignment=VAlign_Center, size {sizeRule=Fill, value=1}}}
  }
  CreateAssetsButtonBox {
   widgetClassPath=&HorizontalBox
   parent=MainControls
   widget=CreateAssetsButtonBox
   slot {class=VerticalBoxSlot, properties {padding=<ltrb 4 4 4 4>}}
  }
  WrapBox_2 {
   widgetClassPath=&WrapBox
   parent=CreateAssetsButtonBox
   widget=WrapBox_2
   properties {horizontalAlignment=HAlign_Fill, innerSlotPadding=<4 4>}
   slot {class=HorizontalBoxSlot, properties {size {sizeRule=Fill, value=1}}}
  }
  CreateAssetsButton {
   widgetClassPath=&EditorUtilityButton
   parent=WrapBox_2
   widget=CreateAssetsButton
   properties {widgetStyle {
     clickedSlateSound {resourceObject=null}
     disabled {
      bIsDynamicallyLoaded=false
      drawAs=RoundedBox
      imageType=NoImage
      mirroring=NoMirror
      resourceName=None
      resourceObject=null
      tiling=NoTile
      imageSize=<0 0>
      margin=<ltrb 0 0 0 0>
      outlineSettings {
       bUseBrushTransparency=false
       roundingType=FixedRadius
       width=1
       color {colorUseRule=UseColor_ColorTable, specifiedColor=<rgba 0 0 0 0>}
       cornerRadii=<4 4 4 4>
      }
      tintColor {colorUseRule=UseColor_Specified, specifiedColor=<rgba 0.0395 0.0395 0.0395 1>}
      uVRegion {bIsValid=false, max=<0 0>, min=<0 0>}
     }
     disabledForeground {colorUseRule=UseColor_ColorTable, specifiedColor=<rgba 0 0 0 0>}
     hovered {
      bIsDynamicallyLoaded=false
      drawAs=RoundedBox
      imageType=NoImage
      mirroring=NoMirror
      resourceName=None
      resourceObject=null
      tiling=NoTile
      imageSize=<32 32>
      margin=<ltrb 0 0 0 0>
      outlineSettings {
       bUseBrushTransparency=false
       roundingType=FixedRadius
       width=1
       color {colorUseRule=UseColor_ColorTable, specifiedColor=<rgba 0 0 0 0>}
       cornerRadii=<4 4 4 4>
      }
      tintColor {colorUseRule=UseColor_Specified, specifiedColor=<rgba 0.0044 0.2384 1 1>}
      uVRegion {bIsValid=false, max=<0 0>, min=<0 0>}
     }
     hoveredForeground {colorUseRule=UseColor_ColorTable, specifiedColor=<rgba 0 0 0 0>}
     hoveredSlateSound {resourceObject=null}
     normal {
      bIsDynamicallyLoaded=false
      drawAs=RoundedBox
      imageType=NoImage
      mirroring=NoMirror
      resourceName=None
      resourceObject=null
      tiling=NoTile
      imageSize=<32 32>
      margin=<ltrb 0 0 0 0>
      outlineSettings {
       bUseBrushTransparency=false
       roundingType=FixedRadius
       width=1
       color {colorUseRule=UseColor_ColorTable, specifiedColor=<rgba 0 0 0 0>}
       cornerRadii=<4 4 4 4>
      }
      tintColor {colorUseRule=UseColor_Specified, specifiedColor=<rgba 0 0.162 0.7454 1>}
      uVRegion {bIsValid=false, max=<0 0>, min=<0 0>}
     }
     normalForeground {colorUseRule=UseColor_ColorTable, specifiedColor=<rgba 0 0 0 0>}
     normalPadding=<ltrb 12 1.5 12 1.5>
     pressed {
      bIsDynamicallyLoaded=false
      drawAs=RoundedBox
      imageType=NoImage
      mirroring=NoMirror
      resourceName=None
      resourceObject=null
      tiling=NoTile
      imageSize=<32 32>
      margin=<ltrb 0 0 0 0>
      outlineSettings {
       bUseBrushTransparency=false
       roundingType=FixedRadius
       width=1
       color {colorUseRule=UseColor_ColorTable, specifiedColor=<rgba 0 0 0 0>}
       cornerRadii=<4 4 4 4>
      }
      tintColor {colorUseRule=UseColor_Specified, specifiedColor=<rgba 0 0.0802 0.3515 1>}
      uVRegion {bIsValid=false, max=<0 0>, min=<0 0>}
     }
     pressedForeground {colorUseRule=UseColor_ColorTable, specifiedColor=<rgba 0 0 0 0>}
     pressedPadding=<ltrb 12 2.5 12 0.5>
     pressedSlateSound {resourceObject=null}
    }}
   slot {class=WrapBoxSlot}
   bIsVariable=~
  }
  TextBlock_295 {
   widgetClassPath=&TextBlock
   parent=CreateAssetsButton
   widget=TextBlock_295
   properties {
    text=Create Assets
    font {
     bForceMonospaced=false
     bMaterialIsStencil=false
     fontMaterial=null
     fontObject=&Roboto
     letterSpacing=0
     monospacedWidth=1
     size=10.5
     skewAmount=0
     typefaceFontName=Bold
     outlineSettings {
      bApplyOutlineToDropShadows=false
      bMiteredCorners=false
      bSeparateFillAlpha=false
      outlineMaterial=null
      outlineSize=0
      outlineColor=<rgba 0 0 0 1>
     }
    }
    margin=<ltrb 0 1 0 0>
   }
   slot {class=ButtonSlot}
  }
  Spacer_66 {widgetClassPath=&Spacer, parent=HorizontalBox_0, widget=Spacer_66, properties {size=<4 1>}, slot {class=HorizontalBoxSlot}}
  DialogueEditor {
   widgetClassPath=/Game/Sounds/Fort_Audio_Blueprints/Utilities/EditorUtilityWidgets/VO_Utilities/Widgets/FortSoundSequenceDialogueViewer
   bIsVariable=true
   parent=HorizontalBox_0
   widget=DialogueEditor
   slot {class=HorizontalBoxSlot, properties {size {sizeRule=Fill, value=1}}}
  }
 ]
}
