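# Editor-only function library of dialogue audio pipeline tools for Fortnite voice-over assets.
# Embedded Python nodes batch-import new WAVs while mirroring subfolders, create Dialogue Waves with speaker, sound and subtitle, and build Sound Cues from them.
# Locate functions search the Content Browser path or a given path recursively through the Asset Registry, filtering by class and name suffix.
# Another function fills a Fort Sound Sequence from dialogue data sounds and post-sentence delays.
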EditorUtilityBlueprint /Game/Sounds/Fort_Audio_Blueprints/Utilities/EditorUtilityWidgets/VO_Utilities/EFL_FortDialogueTools
graphs [
 Create Dialogue Wave Asset {nodes [
   n1 = Create Dialogue Wave Asset!() -> (
    then
    Destination Path:String
    Asset Name:String
    Dialogue Voice:Dialogue Voice Object Reference
    Sound Wave:Soft Object Reference
    Subtitle:String
   ) {is=title K2Node_FunctionEntry}
   > n2 = Execute Python Script!(
    execute
    PythonScript="import unreal\r\n\r\nASSET_TOOLS = unreal.AssetToolsHelpers.get_asset_tools()\r\nDIALOGUE_WAVE_FACTORY = unreal.DialogueWaveFactory()\r\n\r\nDIALOGUE_WAVE_FACTORY.set_editor_property('editor_import', True)\r\n\r\ndialogue_wave = ASSET_TOOLS.create_asset(asset_name\r\n\t, destination_path\r\n\t, unreal.DialogueWave\r\n\t, DIALOGUE_WAVE_FACTORY)\r\n\r\ncontext = unreal.DialogueContext()\r\ncontext.set_editor_property('speaker', speaker)\r\ncontext.set_editor_property('targets', [])\r\n\r\ncontext_mapping = unreal.DialogueContextMapping()\r\ncontext_mapping.set_editor_property('context', context)\r\ncontext_mapping.set_editor_property('sound_wave', sound)\r\n\r\ncontext_mapping_array = []\r\ncontext_mapping_array.append(context_mapping)\r\n\r\ndialogue_wave.set_editor_property('context_mappings', context_mapping_array)\r\ndialogue_wave.set_editor_property('spoken_text', subtitle)"
    destination_path:String=@n1
    asset_name:String=@n1.Asset Name
    subtitle:String=@n1.Subtitle
    speaker:Dialogue Voice Object Reference=@n1.Dialogue Voice
    sound:Soft Object Reference=@n1.Sound Wave
   ) -> (then, ReturnValue) {inputs [destination_path, asset_name, subtitle, speaker, sound]}
   > Return Node!(execute, Success:bool=@n2)
  ]}
 Create Sound Cue from Dialogue Wave {nodes [
   n1 = Create Sound Cue from Dialogue Wave!() -> (then, Dialogue Wave:Soft Object Reference, Destination Path:String) {is=title K2Node_FunctionEntry}
   > Execute Python Script!(
    execute
    PythonScript="import unreal\r\n\r\nsound_cue_factory = unreal.SoundCueFactoryNew()\r\nsound_cue_factory.set_editor_property('editor_import', True)\r\nsound_cue_factory.set_editor_property('InitialDialogueWaves', dialogue_wave)\r\n\r\nasset_tools = unreal.AssetToolsHelpers.get_asset_tools()\r\nasset_tools.create_asset(asset_name, destination_path, unreal.SoundCue, sound_cue_factory)"
    destination_path=@n1.Destination Path
    asset_name=Append(A=GetSoftPtrAssetName!(Asset:Soft Object Reference=@n1) -> (ReturnValue:String), B=_Cue)
    dialogue_wave:Array of Soft Object References=Make Array("[0]":Soft Object Reference=@n1) -> (Array:Array of Soft Object References)
   ) -> (then, ReturnValue) {inputs [destination_path, asset_name, dialogue_wave]}
  ]}
 Filter Found Assets {nodes [
   n1 = Filter Found Assets!() -> (then, Class:Object Class Reference, Filename Suffix:String, Found Asset Data:Array of Asset Data Structures) {is=title K2Node_FunctionEntry}
   > n2 = For Each Loop(Array:Array of Asset Data Structures=@n1.Found Asset Data) -> (Array Element:Asset Data Structure)
   n3 = Get Local Found Assets() -> (Local Found Assets:Array of Soft Object References)
   Return Node!(
    execute=@n2.Completed
    Success="integer > integer"(A=Length(TargetArray:Array of Soft Object References=@n3))
    Found Assets:Array of Soft Object References=Get Local Found Assets()
   )
   > n4 = Set Local Asset Data(Local Asset Data:Asset Data Structure=@n2) -> (Output_Get:Asset Data Structure)
   n5 = GetClass!(InAssetData:"Asset Data Structure (by ref)"=@n4) -> (ReturnValue:Object Class Reference)
   n6 = Get Class() -> (Class:Object Class Reference) {is=title K2Node_VariableGet}
   n7 = Get Filename Suffix() -> (Filename Suffix:String)
   n8 = EndsWith(SourceString=GetFullName!(InAssetData:"Asset Data Structure (by ref)"=@n4) -> (ReturnValue:String), InSuffix=@n7)
   n9 = IsEmpty(InString=@n7)
   > Branch(Condition=AND Boolean(A=ClassIsChildOf(TestClass=@n5, ParentClass=@n6), B=OR Boolean(A=@n8, B=@n9)))
   n10 = Get Local Asset Data() -> (Local Asset Data:Asset Data Structure)
   > Add(
    TargetArray:Array of Soft Object References=Get Local Found Assets()
    NewItem:"Soft Object Reference (by ref)"=To Soft Object Reference(SoftObjectPath=ToSoftObjectPath!(InAssetData:"Asset Data Structure (by ref)"=@n10) -> (ReturnValue:Soft Object Path Structure))
   ) {is=title K2Node_CallArrayFunction}
  ]}
 Get Subfolder Path {nodes [
   n1 = Get Subfolder Path!() -> (then, Asset:Object Reference, Subfolder Name:String) {is=title K2Node_FunctionEntry}
   n2 = Get Object Path String(Object=@n1)
   > Return Node!(
    execute
    Subfolder Path:String=Append!(A=Split!(InPath:String=@n2) -> (PathPart:String, FilenamePart:String, ExtensionPart:String), B=/, C:String=@n1.Subfolder Name, D:String=/) -> (ReturnValue)
   )
  ]}
 Import Sound Waves {nodes [
   n1 = Import Sound Waves!() -> (then, Source Folder:String, Destination Folder:String) {is=title K2Node_FunctionEntry}
   > Execute Python Script!(
    execute
    PythonScript="import unreal\r\nimport os\r\n\r\ntasks = []\r\n\r\ndef queue_sound(wav_path, game_path, asset_name):\r\n    \"\"\"\r\n    Import a single audio file into the engine provided a wav file.\r\n    :param str wav_path: Path to wav.\r\n    :param str game_path: Game path asset location.\r\n    :param str asset_name: Name of asset.\r\n    :return:\r\n    \"\"\"\r\n    ## Create an import task.\r\n    import_task = unreal.AssetImportTask()\r\n\r\n    ## Set base properties on the import task.\r\n    import_task.filename = wav_path\r\n    import_task.destination_path = game_path\r\n    import_task.destination_name = asset_name\r\n    import_task.replace_existing = True\r\n    import_task.automated = True  ## Suppress UI.\r\n\r\n    tasks.append(import_task)\r\n\r\n    ##if not imported_assets:\r\n    ##    unreal.log_warning(wav_path + \" was not imported!\")\r\n    ##    return\r\n\r\n    ## Return the instance of the imported wav\r\n    ##return unreal.load_asset(imported_assets[0])\r\n\r\ndef import_sounds():\r\n    ## Import the wavs.\r\n    unreal.AssetToolsHelpers.get_asset_tools().import_asset_tasks(tasks)\r\n    ##imported_assets = import_task.get_editor_property(\r\n    ##    \"imported_object_paths\"\r\n    ##)\r\n\r\ndef walk_directory(directory):\r\n    \"\"\"\r\n    Opens the ReaperRenderedFileList.txt file located in this file's directory\r\n    and runs import_sound on each line in that text file.\r\n    \"\"\"\r\n    \r\n    if not os.path.exists(directory):\r\n        unreal.log(directory + \" is not a valid path!\")\r\n        return\r\n\r\n    for (subdir, dirs, files) in os.walk(directory):\r\n        for file in files:\r\n            asset_name = os.path.splitext(file)[0]\r\n            rel_dir = (os.path.relpath(subdir, directory))\r\n            file_path = (os.path.join(subdir, file))\r\n            rel_game_path = (os.path.join(game_path, rel_dir)).replace(\"\\\\\",\"/\")\r\n            filename = (os.path.join(rel_game_path, asset_name).replace(\"\\\\\",\"/\"))\r\n            if os.path.splitext(file)[-1].lower() != '.wav' or unreal.EditorAssetLibrary.does_asset_exist(filename):\r\n                print(filename + ' already exists')\r\n                continue\r\n            queue_sound(file_path, rel_game_path, asset_name)\r\n\r\n    import_sounds()\r\n\r\n## Import sounds from given directory and subdirectories.\r\nwalk_directory(root_dir)\r\n"
    root_dir:String=@n1
    game_path:String=@n1.Destination Folder
   ) -> (then, ReturnValue) {inputs [root_dir,game_path]}
  ]}
 Locate Assets in Current Content Browser {nodes [
   Locate Assets in Current Content Browser!() -> (then, Class:Object Class Reference, Filename String:String) {is=title K2Node_FunctionEntry}
   n1 = GetCurrentContentBrowserPath!() -> (OutPath:String, ReturnValue:bool)
   > n2 = Branch(Condition=@n1.ReturnValue)
   > n3 = Locate Assets in Path!(
    execute
    Path:Name=String To Name(InString=@n1)
    Class:Object Class Reference=Get Class()
    Filename Suffix:String=Get Filename String() -> (Filename String:String)
    __WorldContext:Object Reference
   ) -> (then, Success:bool, Found Assets:Array of Soft Object References)
   > Return Node!(execute, Success:bool=@n3, Found Assets:Array of Soft Object References=@n3.Found Assets)
   Return Node(execute=@n2.else, Success=false)
  ]}
 Locate Assets in Path {nodes [
   n1 = Locate Assets in Path!() -> (then, Path:Name, Class:Object Class Reference, Filename Suffix:String) {is=title K2Node_FunctionEntry}
   > n2 = GetAssetsByPath!(execute, self:Asset Registry Interface=GetAssetRegistry(), PackagePath:Name=@n1, bRecursive:bool=true, bIncludeOnlyOnDiskAssets:bool=false) -> (then, OutAssetData:Array of Asset Data Structures, ReturnValue:bool)
   > n3 = Branch(Condition=@n2.ReturnValue)
   > n4 = GetEditorWorld(self=Get UnrealEditorSubsystem())
   > n5 = Filter Found Assets!(
    execute
    Class:Object Class Reference=Get Class()
    Filename Suffix:String=Get Filename Suffix()
    Found Asset Data:Array of Asset Data Structures=@n2
    __WorldContext:Object Reference=@n4
   ) -> (then, Success:bool, Found Assets:Array of Soft Object References) {is=title K2Node_CallFunction}
   > Return Node(Success=@n5, Found Assets=@n5.Found Assets)
   Return Node(execute=@n3.else, Success=false)
  ]}
 Populate Sound Sequence Data {nodes [
   n1 = Populate Sound Sequence Data!() -> (then, Sound Sequence:Fort Sound Sequence Object Reference, In Sound Sequence Data:Fort Sound Sequence Data Object Reference) {is=title K2Node_FunctionEntry}
   > n2 = For Each Loop(Array:Array of Fort Dialogue Data Object References=Get Dialogue Data!(self:Fort Sound Sequence Data Object Reference=@n1.In Sound Sequence Data) -> (Dialogue Data:Array of Fort Dialogue Data Object References) {is=title K2Node_CallFunction}) -> (Array Element:Fort Dialogue Data Object Reference)
   SetEditorProperty(
    execute=@n2.Completed
    Object=Get Sound Sequence() -> (Sound Sequence:Fort Sound Sequence Object Reference)
    PropertyName=SoundSequenceData
    PropertyValue:Array of Sound Sequence Data Structures=Get Sound Sequence Data() -> (Sound Sequence Data:Array of Sound Sequence Data Structures)
   )
   n3 = Get Sound!(self:Fort Dialogue Data Object Reference=@n2) -> (Sound:Sound Base Object Reference)
   Is Valid(exec=@n2.LoopBody, InputObject=@n3)
   > Add(
    TargetArray:Array of Sound Sequence Data Structures=Get Sound Sequence Data()
    NewItem:"Sound Sequence Data Structure (by ref)"=Make Sound Sequence Data!(
     Sound:Sound Base Object Reference=@n3
     Delay:float=Get Post-Sentence Delay!(self:Fort Dialogue Data Object Reference=@n2) -> (Post-Sentence Delay:double)
    ) -> (SoundSequenceData:Sound Sequence Data Structure) {is=title K2Node_MakeStruct}
   )
  ]}
]
variables []
